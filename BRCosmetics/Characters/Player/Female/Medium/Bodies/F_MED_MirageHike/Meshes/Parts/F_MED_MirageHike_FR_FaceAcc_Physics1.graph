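# Physics asset for the FaceAcc part of the F_MED_MirageHike female character, defining capsule bodies and joints for dynamic hair.
# Five bang chains and two side-hair chains root at the head, with upper spine, clavicle, upper-arm and three head capsules forming the body shapes.
# Body joints and chain-root joints swing and twist freely, while nearly all links within chains have per-joint angular limits.
# The first link of the left 01 bang is also free.
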PhysicsAsset /BRCosmetics/Characters/Player/Female/Medium/Bodies/F_MED_MirageHike/Meshes/Parts/F_MED_MirageHike_FR_FaceAcc_Physics1
bodies {
 clavicle_l [{shapeType=Capsule, length=5.3544, radius=5.3769, center=<6.4482 -0.3686 -2.4355>, rotation=<rot 85.606 0 0>}]
 clavicle_r [{shapeType=Capsule, length=4.8637, radius=5.8676, center=<-6.4482 0.3686 2.4355>, rotation=<rot 85.606 0 0>}]
 dyn_hair_bang_01_center_01 [{shapeType=Capsule, length=2, radius=1, shapeName=dyn_hair_bang_01_center_01_capsule, center=<2 0 0>, rotation=<rot 90 0 0>}]
 dyn_hair_bang_01_center_02 [{shapeType=Capsule, length=6.875, radius=1, shapeName=dyn_hair_bang_01_center_02_capsule, center=<4 0 0>, rotation=<rot 90 0 0>}]
 dyn_hair_bang_01_center_03 [{shapeType=Capsule, length=6, radius=1, shapeName=dyn_hair_bang_01_center_03_capsule, center=<4 0 0>, rotation=<rot 90 0 0>}]
 dyn_hair_bang_01_center_04 [{shapeType=Capsule, length=4, radius=1, shapeName=dyn_hair_bang_01_center_04_capsule, center=<2 0 0>, rotation=<rot 90 0 0>}]
 dyn_hair_bang_01_center_mstr [{shapeType=Capsule, length=0.505, radius=0.505, shapeName=dyn_hair_bang_01_center_mstr_capsule, center=<3 0 0>, rotation=<rot 90 0 0>}]
 dyn_hair_bang_01_left_01 [{
   shapeType=Capsule
   length=1.4112
   radius=0.959
   shapeName=dyn_hair_bang_01_left_01_capsule
   center=<1.7849 0 0>
   rotation=<rot 0 -89.0471 -81.8654>
  }]
 dyn_hair_bang_01_left_02 [{
   shapeType=Capsule
   length=3.8378
   radius=1.5313
   shapeName=dyn_hair_bang_01_left_02_capsule
   center=<4.5 0 0>
   rotation=<rot 0 89.6262 -86.6965>
  }]
 dyn_hair_bang_01_left_03 [{shapeType=Capsule, length=4.8428, radius=1.8398, shapeName=dyn_hair_bang_01_left_03_capsule, center=<5 0 2.5>, rotation=<rot 90 0 0>}]
 dyn_hair_bang_01_left_04 [{shapeType=Capsule, length=3.418, radius=1.2448, shapeName=dyn_hair_bang_01_left_04_capsule, center=<3 0 0>, rotation=<rot 90 0 0>}]
 dyn_hair_bang_01_left_mstr [{shapeType=Capsule, length=0.505, radius=0.505, shapeName=dyn_hair_bang_01_left_mstr_capsule, center=<3 0 0>, rotation=<rot 90 0 0>}]
 dyn_hair_bang_01_right_01 [{shapeType=Capsule, length=2, radius=1, shapeName=dyn_hair_bang_01_right_01_capsule, center=<2 0 0>, rotation=<rot 90 0 0>}]
 dyn_hair_bang_01_right_02 [{shapeType=Capsule, length=5.875, radius=1.125, shapeName=dyn_hair_bang_01_right_02_capsule, center=<4 0 0>, rotation=<rot 90 0 0>}]
 dyn_hair_bang_01_right_03 [{shapeType=Capsule, length=4.625, radius=1.25, shapeName=dyn_hair_bang_01_right_03_capsule, center=<3 0 0>, rotation=<rot 90 0 0>}]
 dyn_hair_bang_01_right_mstr [{shapeType=Capsule, length=0.505, radius=0.505, shapeName=dyn_hair_bang_01_right_mstr_capsule, center=<3 0 0>, rotation=<rot 90 0 0>}]
 dyn_hair_bang_02_left_01 [{shapeType=Capsule, length=2, radius=1, shapeName=dyn_hair_bang_02_left_01_capsule, center=<2 0 0>, rotation=<rot 90 0 0>}]
 dyn_hair_bang_02_left_02 [{shapeType=Capsule, length=4, radius=1, shapeName=dyn_hair_bang_02_left_02_capsule, center=<4 0 0>, rotation=<rot 90 0 0>}]
 dyn_hair_bang_02_left_03 [{shapeType=Capsule, length=3.125, radius=2.875, shapeName=dyn_hair_bang_02_left_03_capsule, center=<4 0 2>, rotation=<rot 90 0 0>}]
 dyn_hair_bang_02_left_04 [{shapeType=Capsule, length=4, radius=1, shapeName=dyn_hair_bang_02_left_04_capsule, center=<3 0 0>, rotation=<rot 90 0 0>}]
 dyn_hair_bang_02_left_mstr [{shapeType=Capsule, length=0.505, radius=0.505, shapeName=dyn_hair_bang_02_left_mstr_capsule, center=<3 0 0>, rotation=<rot 90 0 0>}]
 dyn_hair_bang_02_right_01 [{shapeType=Capsule, length=2, radius=1, shapeName=dyn_hair_bang_02_right_01_capsule, center=<2 0 0>, rotation=<rot 90 0 0>}]
 dyn_hair_bang_02_right_02 [{shapeType=Capsule, length=5.375, radius=2.25, shapeName=dyn_hair_bang_02_right_02_capsule, center=<3 0 1.5>, rotation=<rot 90 0 0>}]
 dyn_hair_bang_02_right_03 [{shapeType=Capsule, length=3, radius=1.75, shapeName=dyn_hair_bang_02_right_03_capsule, center=<4 0 1>, rotation=<rot 90 0 0>}]
 dyn_hair_bang_02_right_04 [{shapeType=Capsule, length=3, radius=1.25, shapeName=dyn_hair_bang_02_right_04_capsule, center=<3 0 0>, rotation=<rot 90 0 0>}]
 dyn_hair_bang_02_right_mstr [{shapeType=Capsule, length=0.505, radius=0.505, shapeName=dyn_hair_bang_02_right_mstr_capsule, center=<3 0 0>, rotation=<rot 90 0 0>}]
 dyn_hair_left_01 [{
   shapeType=Capsule
   length=4.4103
   radius=1.7143
   shapeName=dyn_hair_left_01_capsule
   center=<5.9817 0.3904 -0.3347>
   rotation=<rot 0 -93.427 -86.9112>
  }]
 dyn_hair_left_02 [{
   shapeType=Capsule
   length=1.1022
   radius=2.0847
   shapeName=dyn_hair_left_02_capsule
   center=<1.7807 0 0.5>
   rotation=<rot 64.9999 -179.9998 -179.9999>
  }]
 dyn_hair_left_03 [{shapeType=Capsule, length=1.6856, radius=1.9791, shapeName=dyn_hair_left_03_capsule, center=<2 1 1>, rotation=<rot 85 -179.9997 -179.9998>}]
 dyn_hair_right_01 [{
   shapeType=Capsule
   length=4.6306
   radius=2.1623
   shapeName=dyn_hair_right_01_capsule
   center=<6.1259 0.173 -0.2794>
   rotation=<rot 0 93.5479 -86.7681>
  }]
 dyn_hair_right_02 [{
   shapeType=Capsule
   length=1.7817
   radius=2.7694
   shapeName=dyn_hair_right_02_capsule
   center=<2.814 0 1>
   rotation=<rot 74.9999 179.9998 179.9999>
  }]
 dyn_hair_right_03 [{shapeType=Capsule, length=1.7993, radius=2.3211, shapeName=dyn_hair_right_03_capsule, center=<2 -1 1>, rotation=<rot 80 179.9994 179.9994>}]
 head [
  {shapeType=Capsule, length=3.4546, radius=6.3843, shapeName=head_capsule, center=<2.376 1.6959 0>, rotation=<rot 90 0 5.0001>}
  {shapeType=Capsule, length=5.375, radius=2, shapeName=head_capsule, center=<1.3844 7.4425 0>, rotation=<rot -90 90 -90.0001>}
  {shapeType=Capsule, length=8.25, radius=7.625, shapeName=head_capsule, center=<-0.3601 -3.806 0>, rotation=<rot 90 -175.2364 174.7638>}
 ]
 spine_04 [{shapeType=Capsule, length=9.7013, radius=7.968, center=<0.6532 1.8714 0>}]
 spine_05 [{shapeType=Capsule, length=15.0318, radius=7.5064, center=<2.4257 -1.9164 -0.0001>}]
 upperarm_l [{shapeType=Capsule, length=18.0812, radius=4.1347, center=<10.8056 -0.4936 0.3456>, rotation=<rot 90 0 0>}]
 upperarm_r [{shapeType=Capsule, length=18.0812, radius=4.1347, center=<-10.8056 0.4936 -0.3456>, rotation=<rot 90 0 0>}]
}
constraints [
 {
  bone1Name=spine_05
  bone2Name=spine_04
  swing1LimitDegrees=45
  swing1Motion=Free
  swing2LimitDegrees=45
  swing2Motion=Free
  twistLimitDegrees=45
  twistMotion=Free
 }
 {
  bone1Name=clavicle_r
  bone2Name=spine_05
  swing1LimitDegrees=45
  swing1Motion=Free
  swing2LimitDegrees=45
  swing2Motion=Free
  twistLimitDegrees=45
  twistMotion=Free
 }
 {
  bone1Name=upperarm_r
  bone2Name=clavicle_r
  swing1LimitDegrees=45
  swing1Motion=Free
  swing2LimitDegrees=45
  swing2Motion=Free
  twistLimitDegrees=45
  twistMotion=Free
 }
 {
  bone1Name=clavicle_l
  bone2Name=spine_05
  swing1LimitDegrees=45
  swing1Motion=Free
  swing2LimitDegrees=45
  swing2Motion=Free
  twistLimitDegrees=45
  twistMotion=Free
 }
 {
  bone1Name=upperarm_l
  bone2Name=clavicle_l
  swing1LimitDegrees=45
  swing1Motion=Free
  swing2LimitDegrees=45
  swing2Motion=Free
  twistLimitDegrees=45
  twistMotion=Free
 }
 {
  bone1Name=head
  bone2Name=spine_05
  swing1LimitDegrees=45
  swing1Motion=Free
  swing2LimitDegrees=45
  swing2Motion=Free
  twistLimitDegrees=45
  twistMotion=Free
 }
 {
  bone1Name=dyn_hair_bang_01_center_mstr
  bone2Name=head
  swing1LimitDegrees=45
  swing1Motion=Free
  swing2LimitDegrees=45
  swing2Motion=Free
  twistLimitDegrees=45
  twistMotion=Free
 }
 {
  bone1Name=dyn_hair_bang_01_center_01
  bone2Name=dyn_hair_bang_01_center_mstr
  swing1LimitDegrees=5
  swing2LimitDegrees=5
  twistLimitDegrees=3
 }
 {
  bone1Name=dyn_hair_bang_01_center_02
  bone2Name=dyn_hair_bang_01_center_01
  swing1LimitDegrees=15
  swing2LimitDegrees=15
  twistLimitDegrees=5
 }
 {
  bone1Name=dyn_hair_bang_01_center_03
  bone2Name=dyn_hair_bang_01_center_02
  swing1LimitDegrees=15
  swing2LimitDegrees=15
  twistLimitDegrees=5
 }
 {
  bone1Name=dyn_hair_bang_01_center_04
  bone2Name=dyn_hair_bang_01_center_03
  swing1LimitDegrees=15
  swing2LimitDegrees=15
  twistLimitDegrees=5
 }
 {
  bone1Name=dyn_hair_bang_01_left_mstr
  bone2Name=head
  swing1LimitDegrees=45
  swing1Motion=Free
  swing2LimitDegrees=45
  swing2Motion=Free
  twistLimitDegrees=45
  twistMotion=Free
 }
 {
  bone1Name=dyn_hair_bang_01_right_mstr
  bone2Name=head
  swing1LimitDegrees=45
  swing1Motion=Free
  swing2LimitDegrees=45
  swing2Motion=Free
  twistLimitDegrees=45
  twistMotion=Free
 }
 {bone1Name=dyn_hair_bang_01_right_01, bone2Name=dyn_hair_bang_01_right_mstr, swing1LimitDegrees=5, swing2LimitDegrees=5, twistLimitDegrees=3}
 {bone1Name=dyn_hair_bang_01_right_02, bone2Name=dyn_hair_bang_01_right_01, swing1LimitDegrees=25, swing2LimitDegrees=25, twistLimitDegrees=5}
 {bone1Name=dyn_hair_bang_01_right_03, bone2Name=dyn_hair_bang_01_right_02, swing1LimitDegrees=25, swing2LimitDegrees=25, twistLimitDegrees=5}
 {
  bone1Name=dyn_hair_bang_02_left_mstr
  bone2Name=head
  swing1LimitDegrees=45
  swing1Motion=Free
  swing2LimitDegrees=45
  swing2Motion=Free
  twistLimitDegrees=45
  twistMotion=Free
 }
 {bone1Name=dyn_hair_bang_02_left_01, bone2Name=dyn_hair_bang_02_left_mstr, swing1LimitDegrees=5, swing2LimitDegrees=5, twistLimitDegrees=3}
 {bone1Name=dyn_hair_bang_02_left_02, bone2Name=dyn_hair_bang_02_left_01, swing1LimitDegrees=5, swing2LimitDegrees=5, twistLimitDegrees=5}
 {bone1Name=dyn_hair_bang_02_left_03, bone2Name=dyn_hair_bang_02_left_02, swing1LimitDegrees=15, swing2LimitDegrees=15, twistLimitDegrees=5}
 {bone1Name=dyn_hair_bang_02_left_04, bone2Name=dyn_hair_bang_02_left_03, swing1LimitDegrees=15, swing2LimitDegrees=15, twistLimitDegrees=5}
 {
  bone1Name=dyn_hair_bang_02_right_mstr
  bone2Name=head
  swing1LimitDegrees=45
  swing1Motion=Free
  swing2LimitDegrees=45
  swing2Motion=Free
  twistLimitDegrees=45
  twistMotion=Free
 }
 {bone1Name=dyn_hair_bang_02_right_01, bone2Name=dyn_hair_bang_02_right_mstr, swing1LimitDegrees=5, swing2LimitDegrees=5, twistLimitDegrees=5}
 {bone1Name=dyn_hair_bang_02_right_02, bone2Name=dyn_hair_bang_02_right_01, swing1LimitDegrees=5, swing2LimitDegrees=5, twistLimitDegrees=5}
 {bone1Name=dyn_hair_bang_02_right_03, bone2Name=dyn_hair_bang_02_right_02, swing1LimitDegrees=20, swing2LimitDegrees=20, twistLimitDegrees=20}
 {bone1Name=dyn_hair_bang_02_right_04, bone2Name=dyn_hair_bang_02_right_03, swing1LimitDegrees=20, swing2LimitDegrees=20, twistLimitDegrees=20}
 {bone1Name=dyn_hair_bang_01_left_04, bone2Name=dyn_hair_bang_01_left_03, swing1LimitDegrees=35, swing2LimitDegrees=35, twistLimitDegrees=5}
 {bone1Name=dyn_hair_bang_01_left_02, bone2Name=dyn_hair_bang_01_left_01, swing1LimitDegrees=15, swing2LimitDegrees=15, twistLimitDegrees=5}
 {bone1Name=dyn_hair_bang_01_left_03, bone2Name=dyn_hair_bang_01_left_02, swing1LimitDegrees=35, swing2LimitDegrees=35, twistLimitDegrees=5}
 {
  bone1Name=dyn_hair_bang_01_left_01
  bone2Name=dyn_hair_bang_01_left_mstr
  swing1LimitDegrees=45
  swing1Motion=Free
  swing2LimitDegrees=45
  swing2Motion=Free
  twistLimitDegrees=45
  twistMotion=Free
 }
 {bone1Name=dyn_hair_left_02, bone2Name=dyn_hair_left_01, swing1LimitDegrees=55, swing2LimitDegrees=55, twistLimitDegrees=5}
 {bone1Name=dyn_hair_left_03, bone2Name=dyn_hair_left_02, swing1LimitDegrees=15, swing2LimitDegrees=15, twistLimitDegrees=5}
 {
  bone1Name=dyn_hair_left_01
  bone2Name=head
  swing1LimitDegrees=45
  swing1Motion=Free
  swing2LimitDegrees=45
  swing2Motion=Free
  twistLimitDegrees=45
  twistMotion=Free
 }
 {bone1Name=dyn_hair_right_02, bone2Name=dyn_hair_right_01, swing1LimitDegrees=55, swing2LimitDegrees=55, twistLimitDegrees=5}
 {bone1Name=dyn_hair_right_03, bone2Name=dyn_hair_right_02, swing1LimitDegrees=15, swing2LimitDegrees=15, twistLimitDegrees=5}
 {
  bone1Name=dyn_hair_right_01
  bone2Name=head
  swing1LimitDegrees=45
  swing1Motion=Free
  swing2LimitDegrees=45
  swing2Motion=Free
  twistLimitDegrees=45
  twistMotion=Free
 }
]
