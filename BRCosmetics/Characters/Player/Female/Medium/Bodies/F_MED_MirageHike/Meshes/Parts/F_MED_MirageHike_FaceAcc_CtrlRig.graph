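# Control Rig for the MirageHike female face accessory that applies corrective offsets to hair bones as the head and arms move.
# Spherical pose readers on head, right upper arm and clavicle produce alphas, some stored and multiplied into combined poses like look-up while aiming two-handed.
# These alphas weight layered ModifyTransforms on back, middle and side hair bones, run through nested sequences.
# An Emote variable scales an extra set of correctives.
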ControlRigBlueprint /BRCosmetics/Characters/Player/Female/Medium/Bodies/F_MED_MirageHike/Meshes/Parts/F_MED_MirageHike_FaceAcc_CtrlRig
graphs [
 ContainedGraph {nodes [
   n1 = RigVMFunctionEntryNode()
   SequenceExecution_12 = &RigVMFunction_Sequence(ExecuteContext=@n1.ExecuteContext)
   SequenceExecution_12_1 = &RigVMFunction_Sequence(ExecuteContext=@SequenceExecution_12.B)
   SequenceExecution_12_2 = &RigVMFunction_Sequence(ExecuteContext=@SequenceExecution_12_1.B)
   RigVMFunctionReturnNode(A=@SequenceExecution_12.A, B=@SequenceExecution_12_1.A, C=@SequenceExecution_12_2.A, D=@SequenceExecution_12_2.B)
  ]}
 ContainedGraph {nodes [
   n1 = RigVMFunctionEntryNode()
   SequenceExecution_1 = &RigVMFunction_Sequence(ExecuteContext=@n1.ExecuteContext)
   SequenceExecution_1_1 = &RigVMFunction_Sequence(ExecuteContext=@SequenceExecution_1.B)
   SequenceExecution_1_2 = &RigVMFunction_Sequence(ExecuteContext=@SequenceExecution_1_1.B)
   RigVMFunctionReturnNode(A=@SequenceExecution_1.A, B=@SequenceExecution_1_1.A, C=@SequenceExecution_1_2.A, D=@SequenceExecution_1_2.B)
  ]}
 ContainedGraph {nodes [
   n1 = RigVMFunctionEntryNode()
   SequenceExecution_4 = &RigVMFunction_Sequence(ExecuteContext=@n1.ExecuteContext)
   SequenceExecution_4_1 = &RigVMFunction_Sequence(ExecuteContext=@SequenceExecution_4.B)
   SequenceExecution_4_2 = &RigVMFunction_Sequence(ExecuteContext=@SequenceExecution_4_1.B)
   RigVMFunctionReturnNode(A=@SequenceExecution_4.A, B=@SequenceExecution_4_1.A, C=@SequenceExecution_4_2.A, D=@SequenceExecution_4_2.B)
  ]}
 ContainedGraph {nodes [
   n1 = RigVMFunctionEntryNode()
   SequenceExecution_3 = &RigVMFunction_Sequence(ExecuteContext=@n1.ExecuteContext)
   SequenceExecution_3_1 = &RigVMFunction_Sequence(ExecuteContext=@SequenceExecution_3.B)
   SequenceExecution_3_2 = &RigVMFunction_Sequence(ExecuteContext=@SequenceExecution_3_1.B)
   RigVMFunctionReturnNode(A=@SequenceExecution_3.A, B=@SequenceExecution_3_1.A, C=@SequenceExecution_3_2.A, D=@SequenceExecution_3_2.B)
  ]}
 ContainedGraph {nodes [
   n1 = RigVMFunctionEntryNode()
   SequenceExecution_2 = &RigVMFunction_Sequence(ExecuteContext=@n1.ExecuteContext)
   SequenceExecution_2_1 = &RigVMFunction_Sequence(ExecuteContext=@SequenceExecution_2.B)
   SequenceExecution_2_2 = &RigVMFunction_Sequence(ExecuteContext=@SequenceExecution_2_1.B)
   RigVMFunctionReturnNode(A=@SequenceExecution_2.A, B=@SequenceExecution_2_1.A, C=@SequenceExecution_2_2.A, D=@SequenceExecution_2_2.B)
  ]}
 ContainedGraph {nodes [
   n1 = RigVMFunctionEntryNode()
   SequenceExecution_7 = &RigVMFunction_Sequence(ExecuteContext=@n1.ExecuteContext)
   SequenceExecution_7_1 = &RigVMFunction_Sequence(ExecuteContext=@SequenceExecution_7.B)
   SequenceExecution_7_2 = &RigVMFunction_Sequence(ExecuteContext=@SequenceExecution_7_1.B)
   RigVMFunctionReturnNode(A=@SequenceExecution_7.A, B=@SequenceExecution_7_1.A, C=@SequenceExecution_7_2.A, D=@SequenceExecution_7_2.B)
  ]}
 ContainedGraph {nodes [
   n1 = RigVMFunctionEntryNode()
   SequenceExecution_6_1 = &RigVMFunction_Sequence(ExecuteContext=@n1.ExecuteContext)
   SequenceExecution_6_1_1 = &RigVMFunction_Sequence(ExecuteContext=@SequenceExecution_6_1.B)
   SequenceExecution_6_1_2 = &RigVMFunction_Sequence(ExecuteContext=@SequenceExecution_6_1_1.B)
   RigVMFunctionReturnNode(A=@SequenceExecution_6_1.A, B=@SequenceExecution_6_1_1.A, C=@SequenceExecution_6_1_2.A, D=@SequenceExecution_6_1_2.B)
  ]}
 ContainedGraph {nodes [
   n1 = RigVMFunctionEntryNode()
   SequenceExecution_9 = &RigVMFunction_Sequence(ExecuteContext=@n1.ExecuteContext)
   SequenceExecution_9_1 = &RigVMFunction_Sequence(ExecuteContext=@SequenceExecution_9.B)
   SequenceExecution_9_2 = &RigVMFunction_Sequence(ExecuteContext=@SequenceExecution_9_1.B)
   RigVMFunctionReturnNode(A=@SequenceExecution_9.A, B=@SequenceExecution_9_1.A, C=@SequenceExecution_9_2.A, D=@SequenceExecution_9_2.B)
  ]}
 ContainedGraph {nodes [
   n1 = RigVMFunctionEntryNode()
   SequenceExecution_8 = &RigVMFunction_Sequence(ExecuteContext=@n1.ExecuteContext)
   SequenceExecution_8_1 = &RigVMFunction_Sequence(ExecuteContext=@SequenceExecution_8.B)
   SequenceExecution_8_2 = &RigVMFunction_Sequence(ExecuteContext=@SequenceExecution_8_1.B)
   RigVMFunctionReturnNode(A=@SequenceExecution_8.A, B=@SequenceExecution_8_1.A, C=@SequenceExecution_8_2.A, D=@SequenceExecution_8_2.B)
  ]}
 ContainedGraph {nodes [
   n1 = RigVMFunctionEntryNode()
   SequenceExecution = &RigVMFunction_Sequence(ExecuteContext=@n1.ExecuteContext)
   SequenceExecution_1 = &RigVMFunction_Sequence(ExecuteContext=@SequenceExecution.B)
   SequenceExecution_2 = &RigVMFunction_Sequence(ExecuteContext=@SequenceExecution_1.B)
   RigVMFunctionReturnNode(A=@SequenceExecution.A, B=@SequenceExecution_1.A, C=@SequenceExecution_2.A, D=@SequenceExecution_2.B)
  ]}
 RigVMFunctionLibrary {links [], nodes []}
 RigVMModel {nodes [
   n1 = RigVMRerouteNode(Value=())
   &RigUnit_ModifyTransforms(
    ItemToModify=((Item=(Type=Bone,Name="dfrm_hair_mstr"),Transform=(Rotation=(X=0,Y=0,Z=0,W=1),Translation=(X=0,Y=0,Z=0),Scale3D=(X=1,Y=1,Z=1))))
    ItemToModify.0.Transform.Rotation=&RigVMFunction_MathQuaternionFromEuler(Euler=(X=0,Y=0,Z=0)).Result
    ExecutePin=@n1.Value
    Weight=RigVMVariableNode() {variable=look_down_upperarm_r_back_multiplied_alpha}.Value
   )
   RigUnit_BeginExecution = &RigUnit_BeginExecution()
   SequenceExecution_12 = RigVMAggregateNode(ExecuteContext=@RigUnit_BeginExecution.ExecutePin)
   SequenceExecution_1 = RigVMAggregateNode(ExecuteContext=@SequenceExecution_12.B)
   SequenceExecution = RigVMAggregateNode(ExecuteContext=@SequenceExecution_1.A)
   n2 = RigVMVariableNode() {variable=head_SPR_look_up_alpha}
   &RigUnit_ModifyTransforms(
    ItemToModify=((Item=(Type=Bone,Name="dyn_hair_back_02_r"),Transform=(Rotation=(X=0,Y=0,Z=0,W=1),Translation=(X=2,Y=0,Z=2),Scale3D=(X=1,Y=1,Z=1))))
    ItemToModify.0.Transform.Rotation=&RigVMFunction_MathQuaternionFromEuler(Euler=(X=0,Y=15,Z=0)).Result
    ExecutePin=@SequenceExecution.B
    Weight=@n2.Value
   )
   > &RigUnit_ModifyTransforms(
    ItemToModify=((Item=(Type=Bone,Name="dyn_hair_back_03_mstr_r"),Transform=(Rotation=(X=0,Y=0,Z=0,W=1),Translation=(X=0,Y=0,Z=0),Scale3D=(X=1,Y=1,Z=1))))
    ItemToModify.0.Transform.Rotation=&RigVMFunction_MathQuaternionFromEuler(Euler=(X=0,Y=0,Z=0)).Result
    Weight=@n2.Value
   )
   n3 = RigVMVariableNode() {variable=head_SPR_look_up_alpha}
   &RigUnit_ModifyTransforms(
    ItemToModify=((Item=(Type=Bone,Name="dyn_hair_back_02_l"),Transform=(Rotation=(X=0,Y=0,Z=0,W=1),Translation=(X=1,Y=0,Z=2),Scale3D=(X=1,Y=1,Z=1))))
    ItemToModify.0.Transform.Rotation=&RigVMFunction_MathQuaternionFromEuler(Euler=(X=0,Y=15,Z=0)).Result
    ExecutePin=@SequenceExecution.D
    Weight=@n3.Value
   )
   > &RigUnit_ModifyTransforms(
    ItemToModify=((Item=(Type=Bone,Name="dyn_hair_back_03_mstr_l"),Transform=(Rotation=(X=0,Y=0,Z=0,W=1),Translation=(X=0,Y=0,Z=0),Scale3D=(X=1,Y=1,Z=1))))
    ItemToModify.0.Transform.Rotation=&RigVMFunction_MathQuaternionFromEuler(Euler=(X=0,Y=0,Z=0)).Result
    Weight=@n3.Value
   )
   n4 = RigVMVariableNode() {variable=head_SPR_look_up_alpha}
   &RigUnit_ModifyTransforms(
    ItemToModify=((Item=(Type=Bone,Name="dyn_hair_mid_mstr"),Transform=(Rotation=(X=0,Y=0,Z=0,W=1),Translation=(X=0,Y=-1,Z=3),Scale3D=(X=1,Y=1,Z=1))))
    ItemToModify.0.Transform.Rotation=&RigVMFunction_MathQuaternionFromEuler(Euler=(X=5,Y=0,Z=0)).Result
    ExecutePin=@SequenceExecution.C
    Weight=@n4.Value
   )
   > &RigUnit_ModifyTransforms(
    ItemToModify=((Item=(Type=Bone,Name="dyn_hair_03_mstr_mid"),Transform=(Rotation=(X=0,Y=0,Z=0,W=1),Translation=(X=0,Y=0,Z=0),Scale3D=(X=1,Y=1,Z=1))))
    ItemToModify.0.Transform.Rotation=&RigVMFunction_MathQuaternionFromEuler(Euler=(X=0,Y=0,Z=0)).Result
    Weight=@n4.Value
   )
   &RigUnit_ModifyTransforms(
    ItemToModify="((Item=(Type=Bone,Name=\"dyn_hair_mid_mstr\"),Transform=(Rotation=(X=0,Y=0,Z=0,W=1),Translation=(X=0,Y=0,Z=0),Scale3D=(X=1,Y=1,Z=1))),(Item=(Type=Bone,Name=\"dyn_hair_back_mstr_l\"),Transform=(Rotation=(X=0,Y=0,Z=0,W=1),Translation=(X=0,Y=-0.5,Z=1),Scale3D=(X= ...(truncated)"
    ItemToModify.0.Transform.Rotation=&RigVMFunction_MathQuaternionFromEuler(Euler=(X=5,Y=0,Z=0)).Result
    ItemToModify.2.Transform.Rotation=&RigVMFunction_MathQuaternionFromEuler(Euler=(X=5,Y=0,Z=5)).Result
    ItemToModify.1.Transform.Rotation=&RigVMFunction_MathQuaternionFromEuler(Euler=(X=5,Y=0,Z=0)).Result
    Weight=RigVMVariableNode() {variable=head_SPR_look_up_alpha}.Value
    ExecutePin=@SequenceExecution.A
   )
   > &RigUnit_ModifyTransforms(
    ItemToModify="((Item=(Type=Bone,Name=\"dyn_hair_back_01_l\"),Transform=(Rotation=(X=0,Y=0,Z=0,W=1),Translation=(X=0,Y=0,Z=0),Scale3D=(X=1,Y=1,Z=1))),(Item=(Type=Bone,Name=\"dyn_hair_back_01_r\"),Transform=(Rotation=(X=0,Y=0,Z=0,W=1),Translation=(X=0,Y=0,Z=0),Scale3D=(X=1,Y= ...(truncated)"
    ItemToModify.0.Transform.Rotation=&RigVMFunction_MathQuaternionFromEuler(Euler=(X=0,Y=-3,Z=0)).Result
    Weight=RigVMVariableNode() {variable=head_SPR_look_up_alpha}.Value
   )
   n5 = /Script/RigVM.RigVMFunction_MathFloatMax(
    B=RigVMVariableNode() {variable=look_up_two_hands_rifle_multiplied_alpha}.Value
    A=RigVMVariableNode() {variable=look_up_and_left_two_hands_multiplied_alpha}.Value
   ) {is=struct RigVMUnitNode, template="Maximum::Execute(in A,in B,out Result)"}
   > &RigUnit_ModifyTransforms(
    ItemToModify=((Item=(Type=Bone,Name="dfrm_hair_mstr"),Transform=(Rotation=(X=0,Y=0,Z=0,W=1),Translation=(X=0,Y=0,Z=0),Scale3D=(X=1,Y=1,Z=1))))
    ItemToModify.0.Transform.Rotation=&RigVMFunction_MathQuaternionFromEuler(Euler=(X=0,Y=0,Z=0)).Result
    Weight=@n5.Result
   )
   > &RigUnit_ModifyTransforms(
    ItemToModify=((Item=(Type=Bone,Name="dyn_hair_back_01_r"),Transform=(Rotation=(X=0,Y=0,Z=0,W=1),Translation=(X=-2,Y=0,Z=0),Scale3D=(X=1,Y=1,Z=1))))
    ItemToModify.0.Transform.Rotation=&RigVMFunction_MathQuaternionFromEuler(Euler=(X=5,Y=0,Z=0)).Result
    Weight=@n5.Result
   )
   > &RigUnit_ModifyTransforms(
    ItemToModify=((Item=(Type=Bone,Name="dyn_hair_back_01_l"),Transform=(Rotation=(X=0,Y=0,Z=0,W=1),Translation=(X=0,Y=0,Z=1),Scale3D=(X=1,Y=1,Z=1))))
    ItemToModify.0.Transform.Rotation=&RigVMFunction_MathQuaternionFromEuler(Euler=(X=0,Y=5,Z=0)).Result
    Weight=@n5.Result
   )
   SequenceExecution_8 = RigVMAggregateNode(ExecuteContext=@SequenceExecution_1.B)
   n6 = RigVMVariableNode() {variable=look_up_and_left_two_hands_multiplied_alpha}
   &RigUnit_ModifyTransforms(
    ItemToModify=((Item=(Type=Bone,Name="dyn_hair_back_mstr_r"),Transform=(Rotation=(X=0,Y=0,Z=0,W=1),Translation=(X=0,Y=-1,Z=0),Scale3D=(X=1,Y=1,Z=1))))
    ItemToModify.0.Transform.Rotation=&RigVMFunction_MathQuaternionFromEuler(Euler=(X=0,Y=0,Z=5)).Result
    ExecutePin=@SequenceExecution_8.A
    Weight=@n6.Value
   )
   > &RigUnit_ModifyTransforms(
    ItemToModify=((Item=(Type=Bone,Name="dyn_hair_back_03_mstr_r"),Transform=(Rotation=(X=0,Y=0,Z=0,W=1),Translation=(X=0,Y=0,Z=0),Scale3D=(X=1,Y=1,Z=1))))
    ItemToModify.0.Transform.Rotation=&RigVMFunction_MathQuaternionFromEuler(Euler=(X=5,Y=0,Z=5)).Result
    Weight=@n6.Value
   )
   n7 = &RigUnit_SphericalPoseReader(
    ActiveRegionScaleFactors=(PositiveWidth=1,NegativeWidth=1,PositiveHeight=0.5,NegativeHeight=1.5)
    ActiveRegionSize=0.1
    Debug=(bDrawDebug=false,DebugScale=25,DebugSegments=20,DebugThickness=0.25)
    DriverAxis=(X=0,Y=1,Z=0)
    DriverItem=(Type=Bone,Name="head")
    FalloffRegionScaleFactors=(PositiveWidth=1.5,NegativeWidth=1.5,PositiveHeight=1,NegativeHeight=1.5)
    FalloffSize=0.2
    FlipWidthScaling=false
    OptionalParentItem=(Type=Bone)
    RotationOffset=(X=55,Y=0,Z=0)
    ExecutePin=@SequenceExecution_8.D
   )
   > &RigUnit_ModifyTransforms(
    ItemToModify="((Item=(Type=Bone,Name=\"dyn_hair_back_02_l\"),Transform=(Rotation=(X=0,Y=0,Z=0,W=1),Translation=(X=0,Y=0,Z=3),Scale3D=(X=1,Y=1,Z=1))),(Item=(Type=Bone,Name=\"dyn_hair_back_03_mstr_l\"),Transform=(Rotation=(X=0,Y=0,Z=0,W=1),Translation=(X=0,Y=0,Z=2),Scale3D=(X ...(truncated)"
    Weight=@n7.OutputParam
   )
   > &RigUnit_ModifyTransforms(
    ItemToModify=((Item=(Type=Bone,Name="dyn_hair_right_01"),Transform=(Rotation=(X=0,Y=0,Z=0,W=1),Translation=(X=0,Y=1,Z=1),Scale3D=(X=1,Y=1,Z=1))))
    Weight=@n7.OutputParam
   )
   > n8 = &RigUnit_SphericalPoseReader(
    ActiveRegionScaleFactors=(PositiveWidth=1,NegativeWidth=1,PositiveHeight=0.5,NegativeHeight=1.5)
    ActiveRegionSize=0.1
    Debug=(bDrawDebug=false,DebugScale=25,DebugSegments=20,DebugThickness=0.25)
    DriverAxis=(X=0,Y=1,Z=0)
    DriverItem=(Type=Bone,Name="head")
    FalloffRegionScaleFactors=(PositiveWidth=1.5,NegativeWidth=1.5,PositiveHeight=1,NegativeHeight=1.5)
    FalloffSize=0.2
    FlipHeightScaling=true
    FlipWidthScaling=false
    OptionalParentItem=(Type=Bone)
    RotationOffset=(X=125,Y=0,Z=0)
   )
   > &RigUnit_ModifyTransforms(
    ItemToModify="((Item=(Type=Bone,Name=\"dyn_hair_back_02_r\"),Transform=(Rotation=(X=0,Y=0,Z=0,W=1),Translation=(X=0,Y=0,Z=2),Scale3D=(X=1,Y=1,Z=1))),(Item=(Type=Bone,Name=\"dyn_hair_back_mstr_r\"),Transform=(Rotation=(X=0,Y=0,Z=0,W=1),Translation=(X=0,Y=0,Z=2),Scale3D=(X=1, ...(truncated)"
    Weight=@n8.OutputParam
   )
   > &RigUnit_ModifyTransforms(
    ItemToModify=((Item=(Type=Bone,Name="dyn_hair_left_01"),Transform=(Rotation=(X=0,Y=0,Z=0,W=1),Translation=(X=0,Y=-1,Z=1),Scale3D=(X=1,Y=1,Z=1))))
    Weight=@n8.OutputParam
   )
   > n9 = &RigUnit_SphericalPoseReader(
    ActiveRegionScaleFactors=(PositiveWidth=2,NegativeWidth=0.5,PositiveHeight=2,NegativeHeight=2)
    ActiveRegionSize=0.075
    Debug=(bDrawDebug=false,DebugScale=25,DebugSegments=20,DebugThickness=0.25)
    DriverAxis=(X=0,Y=1,Z=0)
    DriverItem=(Type=Bone,Name="head")
    FalloffRegionScaleFactors=(PositiveWidth=1,NegativeWidth=0.4,PositiveHeight=1,NegativeHeight=1)
    FalloffSize=0.2
    FlipHeightScaling=false
    FlipWidthScaling=true
    OptionalParentItem=(Type=Bone)
    RotationOffset=(X=90,Y=0,Z=25)
   )
   n10 = &RigVMFunction_MathQuaternionFromEuler(Euler=(X=0,Y=-5,Z=0))
   > &RigUnit_ModifyTransforms(
    ItemToModify="((Item=(Type=Bone,Name=\"dyn_hair_left_01\"),Transform=(Rotation=(X=0,Y=0,Z=0,W=1),Translation=(X=0,Y=0,Z=0),Scale3D=(X=1,Y=1,Z=1))),(Item=(Type=Bone,Name=\"dyn_hair_right_01\"),Transform=(Rotation=(X=0,Y=0,Z=0,W=1),Translation=(X=0,Y=0,Z=0),Scale3D=(X=1,Y=1,Z ...(truncated)"
    ItemToModify.0.Transform.Rotation=@n10.Result
    ItemToModify.1.Transform.Rotation=@n10.Result
    Weight=@n9.OutputParam
   )
   n11 = RigVMVariableNode() {variable=look_up_and_left_two_hands_multiplied_alpha}
   &RigUnit_ModifyTransforms(
    ItemToModify=((Item=(Type=Bone,Name="dyn_hair_mid_mstr"),Transform=(Rotation=(X=0,Y=0,Z=0,W=1),Translation=(X=0,Y=0,Z=0),Scale3D=(X=1,Y=1,Z=1))))
    ItemToModify.0.Transform.Rotation=&RigVMFunction_MathQuaternionFromEuler(Euler=(X=0,Y=5,Z=0)).Result
    Weight=@n11.Value
    ExecutePin=@SequenceExecution_8.B
   )
   > &RigUnit_ModifyTransforms(
    ItemToModify=((Item=(Type=Bone,Name="dyn_hair_03_mstr_mid"),Transform=(Rotation=(X=0,Y=0,Z=0,W=1),Translation=(X=0,Y=0,Z=0),Scale3D=(X=1,Y=1,Z=1))))
    ItemToModify.0.Transform.Rotation=&RigVMFunction_MathQuaternionFromEuler(Euler=(X=0,Y=0,Z=5)).Result
    Weight=@n11.Value
   )
   n12 = RigVMVariableNode() {variable=look_up_and_left_two_hands_multiplied_alpha}
   &RigUnit_ModifyTransforms(
    ItemToModify=((Item=(Type=Bone,Name="dyn_hair_back_mstr_l"),Transform=(Rotation=(X=0,Y=0,Z=0,W=1),Translation=(X=0,Y=0,Z=0),Scale3D=(X=1,Y=1,Z=1))))
    ItemToModify.0.Transform.Rotation=&RigVMFunction_MathQuaternionFromEuler(Euler=(X=5,Y=5,Z=0)).Result
    ExecutePin=@SequenceExecution_8.C
    Weight=@n12.Value
   )
   > &RigUnit_ModifyTransforms(
    ItemToModify=((Item=(Type=Bone,Name="dyn_hair_back_03_mstr_l"),Transform=(Rotation=(X=0,Y=0,Z=0,W=1),Translation=(X=0,Y=0,Z=0),Scale3D=(X=1,Y=1,Z=1))))
    ItemToModify.0.Transform.Rotation=&RigVMFunction_MathQuaternionFromEuler(Euler=(X=0,Y=0,Z=-5)).Result
    Weight=@n12.Value
   )
   SequenceExecution_9 = RigVMAggregateNode(ExecuteContext=@SequenceExecution_1.C)
   SequenceExecution_6_1 = RigVMAggregateNode(ExecuteContext=@SequenceExecution_9.B)
   n13 = RigVMVariableNode() {variable=look_down_upperarm_r_back_multiplied_alpha}
   &RigUnit_ModifyTransforms(
    ItemToModify=((Item=(Type=Bone,Name="dyn_hair_back_mstr_r"),Transform=(Rotation=(X=0,Y=0,Z=0,W=1),Translation=(X=0,Y=0,Z=0),Scale3D=(X=1,Y=1,Z=1))))
    ItemToModify.0.Transform.Rotation=&RigVMFunction_MathQuaternionFromEuler(Euler=(X=5,Y=0,Z=5)).Result
    ExecutePin=@SequenceExecution_6_1.B
    Weight=@n13.Value
   )
   > &RigUnit_ModifyTransforms(
    ItemToModify=((Item=(Type=Bone,Name="dyn_hair_back_03_mstr_r"),Transform=(Rotation=(X=0,Y=0,Z=0,W=1),Translation=(X=0,Y=0,Z=0),Scale3D=(X=1,Y=1,Z=1))))
    ItemToModify.0.Transform.Rotation=&RigVMFunction_MathQuaternionFromEuler(Euler=(X=0,Y=0,Z=5)).Result
    Weight=@n13.Value
   )
   n14 = RigVMVariableNode() {variable=look_down_upperarm_r_back_multiplied_alpha}
   &RigUnit_ModifyTransforms(
    ItemToModify=((Item=(Type=Bone,Name="dyn_hair_mid_mstr"),Transform=(Rotation=(X=0,Y=0,Z=0,W=1),Translation=(X=0,Y=0,Z=0),Scale3D=(X=1,Y=1,Z=1))))
    ItemToModify.0.Transform.Rotation=&RigVMFunction_MathQuaternionFromEuler(Euler=(X=5,Y=0,Z=0)).Result
    ExecutePin=@SequenceExecution_6_1.C
    Weight=@n14.Value
   )
   > &RigUnit_ModifyTransforms(
    ItemToModify=((Item=(Type=Bone,Name="dyn_hair_03_mstr_mid"),Transform=(Rotation=(X=0,Y=0,Z=0,W=1),Translation=(X=0,Y=0,Z=0),Scale3D=(X=1,Y=1,Z=1))))
    ItemToModify.0.Transform.Rotation=&RigVMFunction_MathQuaternionFromEuler(Euler=(X=0,Y=-5,Z=0)).Result
    Weight=@n14.Value
   )
   n15 = RigVMVariableNode() {variable=look_down_upperarm_r_back_multiplied_alpha}
   &RigUnit_ModifyTransforms(
    ItemToModify=((Item=(Type=Bone,Name="dyn_hair_back_mstr_l"),Transform=(Rotation=(X=0,Y=0,Z=0,W=1),Translation=(X=0,Y=0,Z=0),Scale3D=(X=1,Y=1,Z=1))))
    ItemToModify.0.Transform.Rotation=&RigVMFunction_MathQuaternionFromEuler(Euler=(X=0,Y=0,Z=-5)).Result
    Weight=@n15.Value
    ExecutePin=@SequenceExecution_6_1.D
   )
   > &RigUnit_ModifyTransforms(
    ItemToModify=((Item=(Type=Bone,Name="dyn_hair_back_03_mstr_l"),Transform=(Rotation=(X=0,Y=0,Z=0,W=1),Translation=(X=0,Y=0,Z=0),Scale3D=(X=1,Y=1,Z=1))))
    ItemToModify.0.Transform.Rotation=&RigVMFunction_MathQuaternionFromEuler(Euler=(X=0,Y=0,Z=-5)).Result
    Weight=@n15.Value
   )
   n16 = &RigUnit_SphericalPoseReader(
    ActiveRegionScaleFactors=(PositiveWidth=0.7,NegativeWidth=1,PositiveHeight=1,NegativeHeight=1)
    ActiveRegionSize=0.07
    Debug=(bDrawDebug=false,DebugScale=25,DebugSegments=20,DebugThickness=0.25)
    DriverAxis=(X=-1,Y=0,Z=0)
    DriverItem=(Type=Bone,Name="clavicle_r")
    FalloffRegionScaleFactors=(PositiveWidth=1,NegativeWidth=1,PositiveHeight=1,NegativeHeight=1)
    FalloffSize=0.25
    FlipWidthScaling=false
    OptionalParentItem=(Type=Bone,Name="spine_05")
    RotationOffset=(X=-10,Y=125,Z=-5)
    ExecutePin=@SequenceExecution_9.D
   )
   > &RigUnit_ModifyTransforms(
    ItemToModify="((Item=(Type=Bone,Name=\"dyn_hair_back_01_r\"),Transform=(Rotation=(X=0,Y=0,Z=0,W=1),Translation=(X=-4,Y=0,Z=1),Scale3D=(X=1,Y=1,Z=1))),(Item=(Type=Bone,Name=\"dyn_hair_02_mid\"),Transform=(Rotation=(X=0,Y=0,Z=0,W=1),Translation=(X=-2,Y=0,Z=0.5),Scale3D=(X=1,Y ...(truncated)"
    Weight=@n16.OutputParam
   )
   > n17 = &RigUnit_SphericalPoseReader(
    ActiveRegionScaleFactors=(PositiveWidth=1,NegativeWidth=1,PositiveHeight=0.5,NegativeHeight=1)
    ActiveRegionSize=0.05
    Debug=(bDrawDebug=false,DebugScale=25,DebugSegments=20,DebugThickness=0.25)
    DriverAxis=(X=-1,Y=0,Z=0)
    DriverItem=(Type=Bone,Name="clavicle_r")
    FalloffRegionScaleFactors=(PositiveWidth=1,NegativeWidth=1,PositiveHeight=0.7,NegativeHeight=1)
    FalloffSize=0.2
    FlipWidthScaling=false
    OptionalParentItem=(Type=Bone,Name="spine_05")
    RotationOffset=(X=-25,Y=100,Z=0)
   )
   > &RigUnit_ModifyTransforms(
    ItemToModify=((Item=(Type=Bone,Name="dyn_hair_right_02"),Transform=(Rotation=(X=0,Y=0,Z=0,W=1),Translation=(X=0,Y=-1,Z=1),Scale3D=(X=1,Y=1,Z=1))))
    ItemToModify.0.Transform.Rotation=&RigVMFunction_MathQuaternionFromEuler(Euler=(X=15,Y=0,Z=0)).Result
    Weight=@n17.OutputParam
   )
   > n18 = &RigUnit_SphericalPoseReader(
    ActiveRegionScaleFactors=(PositiveWidth=0.5,NegativeWidth=1,PositiveHeight=1.8,NegativeHeight=2)
    ActiveRegionSize=0.08
    Debug=(bDrawDebug=false,bDraw2D=false,bDrawLocalAxes=true,DebugScale=25,DebugSegments=20,DebugThickness=0.25)
    DriverAxis=(X=-1,Y=0,Z=0)
    DriverItem=(Type=Bone,Name="upperarm_r")
    FalloffRegionScaleFactors=(PositiveWidth=0.5,NegativeWidth=1,PositiveHeight=0.5,NegativeHeight=1)
    FalloffSize=0.45
    FlipHeightScaling=false
    FlipWidthScaling=true
    OptionalParentItem=(Type=Bone,Name="spine_04")
    RotationOffset=(X=-80,Y=40,Z=25)
   )
   > &RigUnit_ModifyTransforms(
    ItemToModify=((Item=(Type=Bone,Name="dyn_hair_right_02"),Transform=(Rotation=(X=0,Y=0,Z=0,W=1),Translation=(X=0,Y=-2,Z=1),Scale3D=(X=1,Y=1,Z=1))))
    ItemToModify.0.Transform.Rotation=&RigVMFunction_MathQuaternionFromEuler(Euler=(X=45,Y=0,Z=-35)).Result
    Weight=@n18.OutputParam
   )
   n19 = &RigUnit_SphericalPoseReader(
    ActiveRegionScaleFactors=(PositiveWidth=2,NegativeWidth=0.5,PositiveHeight=2,NegativeHeight=2)
    ActiveRegionSize=0.075
    Debug=(bDrawDebug=false,DebugScale=25,DebugSegments=20,DebugThickness=0.25)
    DriverAxis=(X=1,Y=0,Z=0)
    DriverItem=(Type=Bone,Name="head")
    FalloffRegionScaleFactors=(PositiveWidth=1,NegativeWidth=0.4,PositiveHeight=1,NegativeHeight=1)
    FalloffSize=0.25
    FlipHeightScaling=false
    FlipWidthScaling=true
    OptionalParentItem=(Type=Bone,Name="spine_05")
    RotationOffset=(X=-70,Y=5,Z=0)
    ExecutePin=@SequenceExecution_9.C
   )
   > &RigUnit_ModifyTransforms(
    ItemToModify="((Item=(Type=Bone,Name=\"dyn_hair_back_mstr_r\"),Transform=(Rotation=(X=0,Y=0,Z=0,W=1),Translation=(X=0,Y=-3,Z=5),Scale3D=(X=1,Y=1,Z=1))),(Item=(Type=Bone,Name=\"dyn_hair_back_mstr_l\"),Transform=(Rotation=(X=0,Y=0,Z=0,W=1),Translation=(X=0,Y=-3,Z=3),Scale3D=( ...(truncated)"
    Weight=@n19.OutputParam
   )
   > &RigUnit_ModifyTransforms(
    ItemToModify="((Item=(Type=Bone,Name=\"dyn_hair_back_01_r\"),Transform=(Rotation=(X=0,Y=0,Z=0,W=1),Translation=(X=-4,Y=0,Z=0),Scale3D=(X=1,Y=1,Z=1))),(Item=(Type=Bone,Name=\"dyn_hair_back_01_l\"),Transform=(Rotation=(X=0,Y=0,Z=0,W=1),Translation=(X=-4,Y=0,Z=0),Scale3D=(X=1, ...(truncated)"
    ItemToModify.0.Transform.Rotation=&RigVMFunction_MathQuaternionFromEuler(Euler=(X=0,Y=5,Z=0)).Result
    ItemToModify.1.Transform.Rotation=&RigVMFunction_MathQuaternionFromEuler(Euler=(X=0,Y=-5,Z=0)).Result
    Weight=@n19.OutputParam
   )
   n20 = &RigUnit_SphericalPoseReader(
    ActiveRegionScaleFactors=(PositiveWidth=1,NegativeWidth=1,PositiveHeight=1,NegativeHeight=1)
    ActiveRegionSize=0.1
    Debug=(bDrawDebug=True,bDraw2D=false,bDrawLocalAxes=false,DebugScale=25,DebugSegments=20,DebugThickness=0.25)
    DriverAxis=(X=0,Y=1,Z=0)
    DriverItem=(Type=Bone,Name="head")
    FalloffRegionScaleFactors=(PositiveWidth=2,NegativeWidth=2,PositiveHeight=2,NegativeHeight=2)
    FalloffSize=0.2
    OptionalParentItem=(Type=Bone)
    RotationOffset=(X=180,Y=90,Z=35)
    ExecutePin=@SequenceExecution_12.C
   )
   AlphaInterp = &RigVMFunction_AlphaInterp(
    InterpSpeedDecreasing=10
    InterpSpeedIncreasing=10
    bClampResult=true
    bInterpResult=False
    Value=@n20.OutputParam
    Scale=RigVMVariableNode() {variable=Emote}.Value
   )
   > &RigUnit_ModifyTransforms(
    ItemToModify=((Item=(Type=Bone,Name="dyn_hair_back_02_l"),Transform=(Rotation=(X=0,Y=0,Z=0,W=1),Translation=(X=-10,Y=-3,Z=7),Scale3D=(X=1,Y=1,Z=1))))
    ItemToModify.0.Transform.Rotation=&RigVMFunction_MathQuaternionFromEuler(Euler=(X=0,Y=0,Z=0)).Result
    Weight=@AlphaInterp.Result
   )
   > &RigUnit_ModifyTransforms(
    ItemToModify=((Item=(Type=Bone,Name="dyn_hair_back_03_mstr_l"),Transform=(Rotation=(X=0,Y=0,Z=0,W=1),Translation=(X=0,Y=0,Z=-5),Scale3D=(X=1,Y=1,Z=1))))
    ItemToModify.0.Transform.Rotation=&RigVMFunction_MathQuaternionFromEuler(Euler=(X=0,Y=0,Z=0)).Result
    Weight=@AlphaInterp.Result
   )
   > &RigUnit_ModifyTransforms(
    ItemToModify=((Item=(Type=Bone,Name="dyn_hair_back_02_r"),Transform=(Rotation=(X=0,Y=0,Z=0,W=1),Translation=(X=-3,Y=0,Z=0),Scale3D=(X=1,Y=1,Z=1))))
    ItemToModify.0.Transform.Rotation=&RigVMFunction_MathQuaternionFromEuler(Euler=(X=0,Y=0,Z=0)).Result
    Weight=@AlphaInterp.Result
   )
   > &RigUnit_ModifyTransforms(
    ItemToModify=((Item=(Type=Bone,Name="dyn_hair_back_03_mstr_r"),Transform=(Rotation=(X=0,Y=0,Z=0,W=1),Translation=(X=0,Y=0,Z=5),Scale3D=(X=1,Y=1,Z=1))))
    ItemToModify.0.Transform.Rotation=&RigVMFunction_MathQuaternionFromEuler(Euler=(X=0,Y=0,Z=0)).Result
    Weight=@AlphaInterp.Result
   )
   > n21 = &RigUnit_SphericalPoseReader(
    ActiveRegionScaleFactors=(PositiveWidth=1,NegativeWidth=1,PositiveHeight=1,NegativeHeight=1)
    ActiveRegionSize=0.1
    Debug=(bDrawDebug=true,bDraw2D=false,bDrawLocalAxes=false,DebugScale=25,DebugSegments=20,DebugThickness=0.25)
    DriverAxis=(X=1,Y=0,Z=0)
    DriverItem=(Type=Bone,Name="upperarm_r")
    FalloffRegionScaleFactors=(PositiveWidth=3,NegativeWidth=3,PositiveHeight=3,NegativeHeight=3)
    FalloffSize=0.25
    FlipHeightScaling=true
    FlipWidthScaling=true
    OptionalParentItem=(Type=Bone)
    RotationOffset=(X=30,Y=0,Z=-50)
   )
   n22 = &RigVMFunction_AlphaInterp(
    InterpSpeedDecreasing=10
    InterpSpeedIncreasing=10
    bClampResult=true
    bInterpResult=False
    Value=@n21.OutputParam
    Scale=RigVMVariableNode() {variable=Emote}.Value
   )
   > &RigUnit_ModifyTransforms(
    ItemToModify=((Item=(Type=Bone,Name="dyn_hair_right_02"),Transform=(Rotation=(X=0,Y=0,Z=0,W=1),Translation=(X=0,Y=-5,Z=5),Scale3D=(X=1,Y=1,Z=1))))
    Weight=@n22.Result
   )
   > n23 = &RigUnit_SphericalPoseReader(
    ActiveRegionScaleFactors=(PositiveWidth=1,NegativeWidth=1,PositiveHeight=1,NegativeHeight=1)
    ActiveRegionSize=0.1
    Debug=(bDrawDebug=True,bDraw2D=false,bDrawLocalAxes=false,DebugScale=25,DebugSegments=20,DebugThickness=0.25)
    DriverAxis=(X=1,Y=0,Z=0)
    DriverItem=(Type=Bone,Name="upperarm_r")
    FalloffRegionScaleFactors=(PositiveWidth=1,NegativeWidth=1,PositiveHeight=1,NegativeHeight=1)
    FalloffSize=0.25
    FlipHeightScaling=true
    FlipWidthScaling=true
    OptionalParentItem=(Type=Bone)
    RotationOffset=(X=130,Y=0,Z=-150)
   )
   n24 = &RigVMFunction_AlphaInterp(
    InterpSpeedDecreasing=10
    InterpSpeedIncreasing=10
    bClampResult=true
    bInterpResult=False
    Value=@n23.OutputParam
    Scale=RigVMVariableNode() {variable=Emote}.Value
   )
   > &RigUnit_ModifyTransforms(
    ItemToModify=((Item=(Type=Bone,Name="dyn_hair_back_03_mstr_r"),Transform=(Rotation=(X=0,Y=0,Z=0,W=1),Translation=(X=0,Y=11.5,Z=0),Scale3D=(X=1,Y=1,Z=1))))
    Weight=@n24.Result
    ItemToModify.0.Transform.Rotation=&RigVMFunction_MathQuaternionFromEuler(Euler=(X=0,Y=0,Z=0)).Result
   )
   SequenceExecution_3 = RigVMAggregateNode(ExecuteContext=@SequenceExecution_12.A)
   SequenceExecution_4 = RigVMAggregateNode(ExecuteContext=@SequenceExecution_3.A)
   SequenceExecution_7 = RigVMAggregateNode(ExecuteContext=@SequenceExecution_4.A)
   n25 = &RigUnit_SphericalPoseReader(
    ActiveRegionScaleFactors=(PositiveWidth=1,NegativeWidth=1,PositiveHeight=1,NegativeHeight=1)
    ActiveRegionSize=0.065
    Debug=(bDrawDebug=false,DebugScale=25,DebugSegments=20,DebugThickness=0.25)
    DriverAxis=(X=-1,Y=0,Z=0)
    DriverItem=(Type=Bone,Name="upperarm_r")
    FalloffRegionScaleFactors=(PositiveWidth=1,NegativeWidth=1,PositiveHeight=1,NegativeHeight=1)
    FalloffSize=0.125
    OptionalParentItem=(Type=Bone)
    RotationOffset=(X=60,Y=0,Z=65)
    ExecutePin=@SequenceExecution_7.A
   )
   > RigVMVariableNode(Value=@n25.OutputParam) {variable=upperarm_r_SPR_back_alpha}
   n26 = &RigUnit_SphericalPoseReader(
    ActiveRegionScaleFactors=(PositiveWidth=1,NegativeWidth=1,PositiveHeight=1,NegativeHeight=1)
    ActiveRegionSize=0.075
    Debug=(bDrawDebug=false,DebugScale=25,DebugSegments=20,DebugThickness=0.25)
    DriverAxis=(X=-1,Y=0,Z=0)
    DriverItem=(Type=Bone,Name="upperarm_r")
    FalloffRegionScaleFactors=(PositiveWidth=1,NegativeWidth=1,PositiveHeight=1,NegativeHeight=1)
    FalloffSize=0.2
    OptionalParentItem=(Type=Bone)
    RotationOffset=(X=-55,Y=0,Z=-20)
    ExecutePin=@SequenceExecution_7.D
   )
   > RigVMVariableNode(Value=@n26.OutputParam) {variable=upperarm_r_rifle_up_alpha}
   n27 = &RigUnit_SphericalPoseReader(
    ActiveRegionScaleFactors=(PositiveWidth=1,NegativeWidth=1,PositiveHeight=1,NegativeHeight=1)
    ActiveRegionSize=0.03
    Debug=(bDrawDebug=false,DebugScale=25,DebugSegments=20,DebugThickness=0.25)
    DriverAxis=(X=-1,Y=0,Z=0)
    DriverItem=(Type=Bone,Name="upperarm_r")
    FalloffRegionScaleFactors=(PositiveWidth=1,NegativeWidth=1,PositiveHeight=1,NegativeHeight=1)
    FalloffSize=0.125
    OptionalParentItem=(Type=Bone)
    RotationOffset=(X=-115,Y=0,Z=25)
    ExecutePin=@SequenceExecution_7.B
   )
   > RigVMVariableNode(Value=@n27.OutputParam) {variable=uppearm_r_SPR_aim_up_two_hands_alpha}
   n28 = &RigUnit_SphericalPoseReader(
    ActiveRegionScaleFactors=(PositiveWidth=2,NegativeWidth=2,PositiveHeight=1,NegativeHeight=1)
    ActiveRegionSize=0.05
    Debug=(bDrawDebug=false,DebugScale=25,DebugSegments=20,DebugThickness=0.25)
    DriverAxis=(X=-1,Y=0,Z=0)
    DriverItem=(Type=Bone,Name="upperarm_r")
    FalloffRegionScaleFactors=(PositiveWidth=1,NegativeWidth=1,PositiveHeight=0.75,NegativeHeight=0.75)
    FalloffSize=0.2
    FlipWidthScaling=false
    OptionalParentItem=(Type=Bone)
    RotationOffset=(X=15,Y=150,Z=0)
    ExecutePin=@SequenceExecution_7.C
   )
   > RigVMVariableNode(Value=@n28.OutputParam) {variable=upperarm_r_skydiving}
   SequenceExecution_2 = RigVMAggregateNode(ExecuteContext=@SequenceExecution_4.B)
   n29 = &RigUnit_SphericalPoseReader(
    ActiveRegionScaleFactors=(PositiveWidth=1,NegativeWidth=0.5,PositiveHeight=1.5,NegativeHeight=1.5)
    ActiveRegionSize=0.1
    Debug=(bDrawDebug=false,DebugScale=25,DebugSegments=20,DebugThickness=0.25)
    DriverAxis=(X=0,Y=1,Z=0)
    DriverItem=(Type=Bone,Name="head")
    FalloffRegionScaleFactors=(PositiveWidth=1,NegativeWidth=0.5,PositiveHeight=1.5,NegativeHeight=1.5)
    FalloffSize=0.2
    FlipWidthScaling=false
    OptionalParentItem=(Type=Bone)
    RotationOffset=(X=90,Y=0,Z=-25)
    ExecutePin=@SequenceExecution_2.A
   )
   > RigVMVariableNode(Value=@n29.OutputParam) {variable=head_SPR_look_up_alpha}
   n30 = &RigUnit_SphericalPoseReader(
    ActiveRegionScaleFactors=(PositiveWidth=2,NegativeWidth=0.5,PositiveHeight=2,NegativeHeight=2)
    ActiveRegionSize=0.075
    Debug=(bDrawDebug=false,DebugScale=25,DebugSegments=20,DebugThickness=0.25)
    DriverAxis=(X=0,Y=1,Z=0)
    DriverItem=(Type=Bone,Name="head")
    FalloffRegionScaleFactors=(PositiveWidth=1,NegativeWidth=0.4,PositiveHeight=1,NegativeHeight=1)
    FalloffSize=0.2
    FlipHeightScaling=false
    FlipWidthScaling=true
    OptionalParentItem=(Type=Bone)
    RotationOffset=(X=90,Y=0,Z=12.5)
    ExecutePin=@SequenceExecution_2.B
   )
   > RigVMVariableNode(Value=@n30.OutputParam) {variable=head_SPR_look_down_alpha}
   n31 = &RigUnit_SphericalPoseReader(
    ActiveRegionScaleFactors=(PositiveWidth=0.75,NegativeWidth=0.75,PositiveHeight=1,NegativeHeight=0.5)
    ActiveRegionSize=0.1
    Debug=(bDrawDebug=false,DebugScale=25,DebugSegments=20,DebugThickness=0.25)
    DriverAxis=(X=0,Y=1,Z=0)
    DriverItem=(Type=Bone,Name="head")
    FalloffRegionScaleFactors=(PositiveWidth=1,NegativeWidth=1,PositiveHeight=1,NegativeHeight=0.5)
    FalloffSize=0.2
    OptionalParentItem=(Type=Bone)
    RotationOffset=(X=120,Y=0,Z=0)
    ExecutePin=@SequenceExecution_2.C
   )
   > RigVMVariableNode(Value=@n31.OutputParam) {variable=head_SPR_look_left_alpha}
   n32 = &RigUnit_SphericalPoseReader(
    ActiveRegionScaleFactors=(PositiveWidth=0.75,NegativeWidth=0.75,PositiveHeight=1,NegativeHeight=0.5)
    ActiveRegionSize=0.1
    Debug=(bDrawDebug=false,DebugScale=25,DebugSegments=20,DebugThickness=0.25)
    DriverAxis=(X=0,Y=1,Z=0)
    DriverItem=(Type=Bone,Name="head")
    FalloffRegionScaleFactors=(PositiveWidth=1,NegativeWidth=1,PositiveHeight=1,NegativeHeight=0.5)
    FalloffSize=0.2
    FlipHeightScaling=true
    FlipWidthScaling=false
    OptionalParentItem=(Type=Bone)
    RotationOffset=(X=60,Y=0,Z=0)
    ExecutePin=@SequenceExecution_2.D
   )
   > RigVMVariableNode(Value=@n32.OutputParam) {variable=head_SPR_look_right_alpha}
   RigVMVariableNode(
    Value=&RigVMFunction_MathFloatMul(B=RigVMVariableNode() {variable=uppearm_r_SPR_aim_up_two_hands_alpha}.Value, A=RigVMVariableNode() {variable=head_SPR_look_up_alpha}.Value) {template="Multiply::Execute(in A,in B,out Result)"}.Result
    ExecuteContext=@SequenceExecution_3.B
   ) {variable=look_up_and_left_two_hands_multiplied_alpha}
   RigVMVariableNode(
    Value=&RigVMFunction_MathFloatMul(A=RigVMVariableNode() {variable=head_SPR_look_up_alpha}.Value, B=RigVMVariableNode() {variable=upperarm_r_rifle_up_alpha}.Value) {template="Multiply::Execute(in A,in B,out Result)"}.Result
    ExecuteContext=@SequenceExecution_3.C
   ) {variable=look_up_two_hands_rifle_multiplied_alpha}
   RigVMVariableNode(
    Value=&RigVMFunction_MathFloatMul(B=RigVMVariableNode() {variable=head_SPR_look_down_alpha}.Value, A=RigVMVariableNode() {variable=upperarm_r_SPR_back_alpha}.Value) {template="Multiply::Execute(in A,in B,out Result)"}.Result
    ExecuteContext=@SequenceExecution_3.D
   ) {variable=look_down_upperarm_r_back_multiplied_alpha}
   &RigVMFunction_MathQuaternionFromEuler(Euler=(X=0,Y=0,Z=0))
   &RigUnit_ModifyTransforms(
    ExecutePin=()
    ItemToModify=((Item=(Type=Bone,Name="dyn_hair_back_01_l"),Transform=(Rotation=(X=0,Y=0,Z=0,W=1),Translation=(X=0,Y=0,Z=0),Scale3D=(X=1,Y=1,Z=1))))
    ItemToModify.0.Transform.Rotation=&RigVMFunction_MathQuaternionFromEuler(Euler=(X=0,Y=0,Z=5)).Result
    Weight=@n15.Value
   )
   SequenceAggregate = &RigVMFunction_Sequence(ExecuteContext=())
  ]}
]
hierarchy [
 root {loc=<0 0 0>, rot=<0 0 0 1>}
 pelvis {parent=root, loc=<0 -0.5301 98.0571>, rot=<0.0127 0.707 -0.0127 -0.707>}
 spine_01 {parent=pelvis, loc=<6.1208 -0.9328 0>, rot=<0 0 0.034 -0.9994>}
 spine_02 {parent=spine_01, loc=<6.5996 -0.2095 0>, rot=<0 0 0.0238 -0.9997>}
 spine_03 {parent=spine_02, loc=<6.4914 -0.1237 0>, rot=<0 0 0.0149 -0.9999>}
 spine_04 {parent=spine_03, loc=<6.9598 0.0794 0>, rot=<0 0 -0.0132 -0.9999>}
 spine_05 {parent=spine_04, loc=<6.9987 0 0>, rot=<0 0 -0.0015 -1>}
 neck_01 {parent=spine_05, loc=<11.3138 0.1529 0>, rot=<0 0 -0.2003 -0.9797>}
 neck_02 {parent=neck_01, loc=<4.5627 0 0>, rot=<0 0 0.0504 -0.9987>}
 head {parent=neck_02, loc=<5.3914 0 0>, rot=<0 0 0.115 -0.9934>}
 dfrm_hair_mstr {parent=head, loc=<9.1012 -4.0009 0>, rot=<0.5042 -0.4958 -0.4999 -0.5>}
 dyn_hair_mid_mstr {parent=dfrm_hair_mstr, loc=<0.004 -0.457 -1.5026>, rot=<0 -1 0 0>}
 dyn_hair_01_mid {parent=dyn_hair_mid_mstr, loc=<0 0 0>, rot=<-0.0636 0.0612 -0.7012 -0.7075>}
 dyn_hair_02_mid {parent=dyn_hair_01_mid, loc=<12.379 0 0>, rot=<0.0038 0.0347 -0.0008 -0.9994>}
 dyn_hair_03_mstr_mid {parent=dyn_hair_02_mid, loc=<11.7877 0 0>, rot=<0 0 0 1>}
 dyn_hair_03_mid {parent=dyn_hair_03_mstr_mid, loc=<0 0 0>, rot=<-0.0005 -0.0178 0.0001 -0.9998>}
 dyn_hair_04_mid {parent=dyn_hair_03_mid, loc=<12.2089 0 0>, rot=<-0.0031 -0.0797 0.0004 -0.9968>}
 dyn_hair_05_mid {parent=dyn_hair_04_mid, loc=<12.2908 0 0>, rot=<-0.0236 -0.0912 0.0165 -0.9954>}
 dyn_hair_back_mstr_r {parent=dfrm_hair_mstr, loc=<-5.5634 -0.0945 -0.0008>, rot=<0 -1 0 0>}
 dyn_hair_back_01_r {parent=dyn_hair_back_mstr_r, loc=<0 0 0>, rot=<-0.0709 0.0447 -0.7005 -0.7087>}
 dyn_hair_back_02_r {parent=dyn_hair_back_01_r, loc=<12.6565 0 0>, rot=<-0.1727 0.1337 0.0534 -0.9744>}
 dyn_hair_back_03_mstr_r {parent=dyn_hair_back_02_r, loc=<12.006 0 0>, rot=<0.0859 0 0 -0.9963>}
 dyn_hair_back_03_r {parent=dyn_hair_back_03_mstr_r, loc=<0 0 0>, rot=<0.0289 -0.1516 0.017 -0.9879>}
 dyn_hair_back_04_r {parent=dyn_hair_back_03_r, loc=<12.0961 0 0>, rot=<0.0019 -0.0609 0.044 -0.9972>}
 dyn_hair_back_05_r {parent=dyn_hair_back_04_r, loc=<12.042 0 0>, rot=<-0.0521 -0.1127 -0.0097 -0.9922>}
 dyn_hair_back_04_fr {parent=dyn_hair_back_03_r, loc=<12.9856 -3.3695 -1.1769>, rot=<-0.0924 -0.1317 0.1173 -0.98>}
 dyn_hair_back_05_fr {parent=dyn_hair_back_04_fr, loc=<9.5607 0 0>, rot=<-0.0796 -0.3274 0.0935 -0.9369>}
 dyn_hair_bang_01_left_mstr {parent=dfrm_hair_mstr, loc=<-1.0821 -1.7835 11.6043>, rot=<0 0 0 1>}
 dyn_hair_bang_01_left_01 {parent=dyn_hair_bang_01_left_mstr, loc=<0 0 0>, rot=<-0.5903 0.1883 0.2059 -0.7574>}
 dyn_hair_bang_01_left_02 {parent=dyn_hair_bang_01_left_01, loc=<2.8767 0 0>, rot=<0.034 -0.7395 -0.0759 -0.668>}
 dyn_hair_bang_01_left_03 {parent=dyn_hair_bang_01_left_02, loc=<8.8692 0 0>, rot=<0.0227 -0.2156 -0.0233 -0.9759>}
 dyn_hair_bang_01_left_04 {parent=dyn_hair_bang_01_left_03, loc=<12.105 0 0>, rot=<-0.057 0.1484 -0.2426 -0.957>}
 dyn_hair_bang_02_left_mstr {parent=dfrm_hair_mstr, loc=<-1.0821 -1.7835 11.6043>, rot=<0 0 0 1>}
 dyn_hair_bang_02_left_01 {parent=dyn_hair_bang_02_left_mstr, loc=<0 0 0>, rot=<-0.7038 0.2859 -0.0823 -0.645>}
 dyn_hair_bang_02_left_02 {parent=dyn_hair_bang_02_left_01, loc=<2.9443 0 0>, rot=<-0.0182 -0.6673 0.1665 -0.7257>}
 dyn_hair_bang_02_left_03 {parent=dyn_hair_bang_02_left_02, loc=<8.4625 0 0>, rot=<-0.0079 -0.2301 0.0433 -0.9722>}
 dyn_hair_bang_02_left_04 {parent=dyn_hair_bang_02_left_03, loc=<8.1542 0 0>, rot=<-0.0074 -0.0267 -0.1846 -0.9824>}
 dyn_hair_bang_02_right_mstr {parent=dfrm_hair_mstr, loc=<-1.0821 -1.7835 11.6043>, rot=<0 0 0 1>}
 dyn_hair_bang_02_right_01 {parent=dyn_hair_bang_02_right_mstr, loc=<0 0 0>, rot=<0.0378 -0.5763 0.7899 0.2059>}
 dyn_hair_bang_02_right_02 {parent=dyn_hair_bang_02_right_01, loc=<2.1357 0 0>, rot=<0.0256 -0.509 -0.1592 -0.8455>}
 dyn_hair_bang_02_right_03 {parent=dyn_hair_bang_02_right_02, loc=<8.5822 0 0>, rot=<-0.0261 -0.1942 0.1616 -0.9672>}
 dyn_hair_bang_02_right_04 {parent=dyn_hair_bang_02_right_03, loc=<7.8134 0 0>, rot=<0.1009 -0.1839 0.2549 -0.9439>}
 dyn_hair_bang_01_center_mstr {parent=dfrm_hair_mstr, loc=<-1.0821 -1.7835 11.6043>, rot=<0 0 0 1>}
 dyn_hair_bang_01_center_01 {parent=dyn_hair_bang_01_center_mstr, loc=<0 0 0>, rot=<0.3713 -0.5151 0.6052 0.4802>}
 dyn_hair_bang_01_center_02 {parent=dyn_hair_bang_01_center_01, loc=<2.8732 0 0>, rot=<0.2081 -0.5999 0.0461 -0.7712>}
 dyn_hair_bang_01_center_03 {parent=dyn_hair_bang_01_center_02, loc=<8.4429 0 0>, rot=<-0.1218 -0.3549 0.0644 -0.9247>}
 dyn_hair_bang_01_center_04 {parent=dyn_hair_bang_01_center_03, loc=<8.3738 0 0>, rot=<0.078 -0.1873 0.0465 -0.9781>}
 dyn_hair_bang_01_right_mstr {parent=dfrm_hair_mstr, loc=<-1.0821 -1.7835 11.6043>, rot=<0 0 0 1>}
 dyn_hair_bang_01_right_01 {parent=dyn_hair_bang_01_right_mstr, loc=<0 0 0>, rot=<-0.0259 -0.7056 0.7063 -0.0516>}
 dyn_hair_bang_01_right_02 {parent=dyn_hair_bang_01_right_01, loc=<2.4363 0 0>, rot=<-0.269 -0.56 -0.0071 -0.7836>}
 dyn_hair_bang_01_right_03 {parent=dyn_hair_bang_01_right_02, loc=<8.7785 0 0>, rot=<0.0667 -0.0824 -0.2939 -0.9499>}
 dyn_hair_left_mstr {parent=dfrm_hair_mstr, loc=<5.49 6.7185 3.5383>, rot=<0 0 0 1>}
 dyn_hair_left_01 {parent=dyn_hair_left_mstr, loc=<0 0 0>, rot=<0.0148 -0.0394 -0.683 -0.7292>}
 dyn_hair_left_02 {parent=dyn_hair_left_01, loc=<10.6988 0 0>, rot=<-0.3727 0.2516 -0.0798 -0.8896>}
 dyn_hair_left_03 {parent=dyn_hair_left_02, loc=<3.3535 0 0>, rot=<0.0099 0.2058 0.1331 -0.9694>}
 dyn_hair_right_mstr {parent=dfrm_hair_mstr, loc=<-5.6444 6.6235 3.5376>, rot=<0 0 0 1>}
 dyn_hair_right_01 {parent=dyn_hair_right_mstr, loc=<0 0 0>, rot=<0.0236 -0.0022 -0.692 -0.7215>}
 dyn_hair_right_02 {parent=dyn_hair_right_01, loc=<10.6588 0 0>, rot=<0.4124 0.3501 0.0071 -0.841>}
 dyn_hair_right_03 {parent=dyn_hair_right_02, loc=<6.1102 0 0>, rot=<0.071 0.1771 -0.4105 -0.8917>}
 dyn_hair_back_mstr_l {parent=dfrm_hair_mstr, loc=<5.5642 0 0>, rot=<0 -1 0 0>}
 dyn_hair_back_01_l {parent=dyn_hair_back_mstr_l, loc=<0 0 0>, rot=<-0.039 0.0781 -0.703 -0.7058>}
 dyn_hair_back_02_l {parent=dyn_hair_back_01_l, loc=<12.6412 0 0>, rot=<0.0321 0.1184 -0.0648 -0.9903>}
 dyn_hair_back_03_mstr_l {parent=dyn_hair_back_02_l, loc=<11.9654 0 0>, rot=<0 0 0 1>}
 dyn_hair_back_03_l {parent=dyn_hair_back_03_mstr_l, loc=<0 0 0>, rot=<0.0497 -0.1259 0.0181 -0.9906>}
 dyn_hair_back_04_l {parent=dyn_hair_back_03_l, loc=<12.0147 0 0>, rot=<0.0423 -0.092 -0.0316 -0.9944>}
 dyn_hair_back_05_l {parent=dyn_hair_back_04_l, loc=<11.9574 0 0>, rot=<0.0196 -0.0108 -0.0098 -0.9997>}
 dyn_hair_back_04_fl {parent=dyn_hair_back_03_l, loc=<13.047 4.2082 -0.7543>, rot=<0.1736 -0.0723 -0.1422 -0.9718>}
 dyn_hair_back_05_fl {parent=dyn_hair_back_04_fl, loc=<9.6826 0 0>, rot=<-0.033 -0.108 0.0124 -0.9935>}
 clavicle_l {parent=spine_05, loc=<5.4831 -0.1053 -3.2946>, rot=<0.0465 -0.7289 -0.009 -0.6829>}
 upperarm_l {parent=clavicle_l, loc=<9.8658 -1.2136 -0.4105>, rot=<0.0164 -0.3245 -0.0269 -0.9454>}
 upperarm_twist_01_l {parent=upperarm_l, loc=<7.2135 -0.1134 0>, rot=<0 0 0 1>}
 clavicle_r {parent=spine_05, loc=<5.4831 -0.1053 3.2946>, rot=<0.7289 0.0465 -0.6829 0.009>}
 upperarm_r {parent=clavicle_r, loc=<-9.8658 1.2136 0.4105>, rot=<0.0164 -0.3245 -0.0269 -0.9454>}
 upperarm_twist_01_r {parent=upperarm_r, loc=<-7.2135 0.1134 0>, rot=<0 0 0 -1>}
]
variables [
 Emote {type=double, default=0}
 clavicle_r_SPR_back_alpha {type=double, default=0}
 head_SPR_look_down_alpha {type=double, default=0}
 head_SPR_look_left_alpha {type=double, default=0}
 head_SPR_look_right_alpha {type=double, default=0}
 head_SPR_look_up_alpha {type=double, default=0}
 look_down_upperarm_r_back_multiplied_alpha {type=double, default=0}
 look_up_and_left_two_hands_multiplied_alpha {type=double, default=0}
 look_up_two_hands_rifle_multiplied_alpha {type=double, default=0}
 uppearm_r_SPR_aim_up_two_hands_alpha {type=double, default=0}
 upperarm_r_SPR_back_alpha {type=double, default=0}
 upperarm_r_rifle_up_alpha {type=double, default=0}
 upperarm_r_skydiving {type=double, default=0}
]
