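# Physics asset for the MirageHike medium female cosmetic's FaceAcc mesh, giving capsule bodies to dynamic hair bones and the upper body.
# Spine, clavicle and upper-arm capsules form a torso joined by free-rotation constraints.
# Middle, left and right hair chains hang from spine_05 through tiny master bodies using angle-limited joints, with side chains forking into fl/fr branches.
# Free-rotation cross-links tie both side chains to the middle chain.
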PhysicsAsset /BRCosmetics/Characters/Player/Female/Medium/Bodies/F_MED_MirageHike/Meshes/Parts/F_MED_MirageHike_FaceAcc_Physics
bodies {
 clavicle_l [{shapeType=Capsule, length=3.6044, radius=7.1269, center=<6.4482 -0.3686 -2.4355>, rotation=<rot 85.606 0 0>}]
 clavicle_r [{shapeType=Capsule, length=3.6044, radius=7.1269, center=<-6.4482 0.3686 2.4355>, rotation=<rot 85.606 0 0>}]
 dyn_hair_03_mid [{shapeType=Capsule, length=7.5, radius=6.5, shapeName=dyn_hair_03_mid_capsule, center=<4 0 4>, rotation=<rot 90 0 0>}]
 dyn_hair_03_mstr_mid [{shapeType=Capsule, length=0.505, radius=0.505, shapeName=dyn_hair_03_mstr_mid_capsule, rotation=<rot 0 -90 0>}]
 dyn_hair_04_mid [{shapeType=Capsule, length=7.5, radius=6.5, shapeName=dyn_hair_04_mid_capsule, center=<4 0 4>, rotation=<rot 90 0 0>}]
 dyn_hair_05_mid [{shapeType=Capsule, length=7.5, radius=6.5, shapeName=dyn_hair_05_mid_capsule, center=<4 0 4>, rotation=<rot 90 0 0>}]
 dyn_hair_back_03_l [{shapeType=Capsule, length=8.25, radius=5.75, shapeName=dyn_hair_back_03_l_capsule, center=<4 0 4>, rotation=<rot 90 0 0>}]
 dyn_hair_back_03_mstr_l [{shapeType=Capsule, length=0.505, radius=0.505, shapeName=dyn_hair_back_03_mstr_l_capsule, rotation=<rot 0 -90 0>}]
 dyn_hair_back_03_mstr_r [{shapeType=Capsule, length=0.505, radius=0.505, shapeName=dyn_hair_back_03_mstr_r_capsule, rotation=<rot 0 -90 0>}]
 dyn_hair_back_03_r [{shapeType=Capsule, length=8.25, radius=5.75, shapeName=dyn_hair_back_03_r_capsule, center=<4 0 4>, rotation=<rot 90 0 0>}]
 dyn_hair_back_04_fl [
  {shapeType=Capsule, length=4.75, radius=3.25, shapeName=dyn_hair_back_04_fl_capsule, center=<4 0 0>, rotation=<rot 90 0 0>}
  {
   shapeType=Capsule
   length=6.875
   radius=1.875
   shapeName=dyn_hair_back_04_fl_capsule
   center=<2.2593 4.0881 -0.7976>
   rotation=<rot -75 -145 -180>
  }
 ]
 dyn_hair_back_04_fr [{shapeType=Capsule, length=5, radius=3, shapeName=dyn_hair_back_04_fr_capsule, center=<4 0 0>, rotation=<rot 90 0 0>}]
 dyn_hair_back_04_l [{shapeType=Capsule, length=7.625, radius=6.375, shapeName=dyn_hair_back_04_l_capsule, center=<4 0 5>, rotation=<rot 90 0 0>}]
 dyn_hair_back_04_r [{shapeType=Capsule, length=7.5, radius=6.5, shapeName=dyn_hair_back_04_r_capsule, center=<4 0 4>, rotation=<rot 90 0 0>}]
 dyn_hair_back_05_fl [{shapeType=Capsule, length=6, radius=4.125, shapeName=dyn_hair_back_05_fl_capsule, center=<4 0 0>, rotation=<rot 74.9999 0 -0.0001>}]
 dyn_hair_back_05_fr [{shapeType=Capsule, length=5, radius=3, shapeName=dyn_hair_back_05_fr_capsule, center=<4 0 0>, rotation=<rot 90 0 0>}]
 dyn_hair_back_05_l [{shapeType=Capsule, length=7.625, radius=6.375, shapeName=dyn_hair_back_05_l_capsule, center=<4 0 5.5>, rotation=<rot 90 0 0>}]
 dyn_hair_back_05_r [{shapeType=Capsule, length=7.5, radius=6.5, shapeName=dyn_hair_back_05_r_capsule, center=<4 0 4>, rotation=<rot 90 0 0>}]
 spine_01 [{shapeType=Capsule, length=7.0613, radius=9.6791, center=<0 1.0381 -0.0001>}]
 spine_02 [{shapeType=Capsule, length=5.9705, radius=8.255, center=<0 1.813 -0.0001>}]
 spine_03 [{shapeType=Capsule, length=7.9414, radius=7.7882, center=<0 2.0457 -0.0001>}]
 spine_04 [{shapeType=Capsule, length=9.7013, radius=7.968, center=<0.6532 1.8714 0>}]
 spine_05 [{shapeType=Capsule, length=15.9068, radius=7.7564, center=<2.8332 0.6436 -0.0001>}]
 upperarm_l [{shapeType=Capsule, length=18.0812, radius=4.1347, center=<10.8056 -0.4936 0.3456>, rotation=<rot 90 0 0>}]
 upperarm_r [{shapeType=Capsule, length=18.0812, radius=4.1347, center=<-10.8056 0.4936 -0.3456>, rotation=<rot 90 0 0>}]
}
constraints [
 {
  bone1Name=spine_05
  bone2Name=spine_04
  swing1LimitDegrees=45
  swing1Motion=Free
  swing2LimitDegrees=45
  swing2Motion=Free
  twistLimitDegrees=45
  twistMotion=Free
 }
 {
  bone1Name=clavicle_r
  bone2Name=spine_05
  swing1LimitDegrees=45
  swing1Motion=Free
  swing2LimitDegrees=45
  swing2Motion=Free
  twistLimitDegrees=45
  twistMotion=Free
 }
 {
  bone1Name=upperarm_r
  bone2Name=clavicle_r
  swing1LimitDegrees=45
  swing1Motion=Free
  swing2LimitDegrees=45
  swing2Motion=Free
  twistLimitDegrees=45
  twistMotion=Free
 }
 {
  bone1Name=clavicle_l
  bone2Name=spine_05
  swing1LimitDegrees=45
  swing1Motion=Free
  swing2LimitDegrees=45
  swing2Motion=Free
  twistLimitDegrees=45
  twistMotion=Free
 }
 {
  bone1Name=upperarm_l
  bone2Name=clavicle_l
  swing1LimitDegrees=45
  swing1Motion=Free
  swing2LimitDegrees=45
  swing2Motion=Free
  twistLimitDegrees=45
  twistMotion=Free
 }
 {
  bone1Name=spine_04
  bone2Name=spine_03
  swing1LimitDegrees=45
  swing1Motion=Free
  swing2LimitDegrees=45
  swing2Motion=Free
  twistLimitDegrees=45
  twistMotion=Free
 }
 {
  bone1Name=spine_02
  bone2Name=spine_01
  swing1LimitDegrees=45
  swing1Motion=Free
  swing2LimitDegrees=45
  swing2Motion=Free
  twistLimitDegrees=45
  twistMotion=Free
 }
 {
  bone1Name=spine_03
  bone2Name=spine_02
  swing1LimitDegrees=45
  swing1Motion=Free
  swing2LimitDegrees=45
  swing2Motion=Free
  twistLimitDegrees=45
  twistMotion=Free
 }
 {
  bone1Name=dyn_hair_back_03_mstr_l
  bone2Name=spine_05
  swing1LimitDegrees=45
  swing1Motion=Free
  swing2LimitDegrees=45
  swing2Motion=Free
  twistLimitDegrees=45
  twistMotion=Free
 }
 {bone1Name=dyn_hair_back_03_l, bone2Name=dyn_hair_back_03_mstr_l, swing1LimitDegrees=45, swing2LimitDegrees=45, twistLimitDegrees=15}
 {bone1Name=dyn_hair_back_04_fl, bone2Name=dyn_hair_back_03_l, swing1LimitDegrees=45, swing2LimitDegrees=45, twistLimitDegrees=15}
 {bone1Name=dyn_hair_back_05_fl, bone2Name=dyn_hair_back_04_fl, swing1LimitDegrees=45, swing2LimitDegrees=45, twistLimitDegrees=15}
 {bone1Name=dyn_hair_back_04_l, bone2Name=dyn_hair_back_03_l, swing1LimitDegrees=45, swing2LimitDegrees=45, twistLimitDegrees=45}
 {bone1Name=dyn_hair_back_05_l, bone2Name=dyn_hair_back_04_l, swing1LimitDegrees=45, swing2LimitDegrees=45, twistLimitDegrees=45}
 {
  bone1Name=dyn_hair_back_03_mstr_r
  bone2Name=spine_05
  swing1LimitDegrees=45
  swing1Motion=Free
  swing2LimitDegrees=45
  swing2Motion=Free
  twistLimitDegrees=45
  twistMotion=Free
 }
 {bone1Name=dyn_hair_back_03_r, bone2Name=dyn_hair_back_03_mstr_r, swing1LimitDegrees=45, swing2LimitDegrees=45, twistLimitDegrees=15}
 {bone1Name=dyn_hair_back_04_fr, bone2Name=dyn_hair_back_03_r, swing1LimitDegrees=45, swing2LimitDegrees=45, twistLimitDegrees=15}
 {bone1Name=dyn_hair_back_05_fr, bone2Name=dyn_hair_back_04_fr, swing1LimitDegrees=45, swing2LimitDegrees=45, twistLimitDegrees=15}
 {bone1Name=dyn_hair_back_04_r, bone2Name=dyn_hair_back_03_r, swing1LimitDegrees=45, swing2LimitDegrees=45, twistLimitDegrees=15}
 {bone1Name=dyn_hair_back_05_r, bone2Name=dyn_hair_back_04_r, swing1LimitDegrees=45, swing2LimitDegrees=45, twistLimitDegrees=15}
 {bone1Name=dyn_hair_05_mid, bone2Name=dyn_hair_04_mid, swing1LimitDegrees=45, swing2LimitDegrees=45, twistLimitDegrees=15}
 {bone1Name=dyn_hair_03_mid, bone2Name=dyn_hair_03_mstr_mid, swing1LimitDegrees=45, swing2LimitDegrees=45, twistLimitDegrees=15}
 {bone1Name=dyn_hair_04_mid, bone2Name=dyn_hair_03_mid, swing1LimitDegrees=45, swing2LimitDegrees=45, twistLimitDegrees=15}
 {
  bone1Name=dyn_hair_03_mstr_mid
  bone2Name=spine_05
  swing1LimitDegrees=45
  swing1Motion=Free
  swing2LimitDegrees=45
  swing2Motion=Free
  twistLimitDegrees=45
  twistMotion=Free
 }
 {
  bone1Name=dyn_hair_back_04_r
  bone2Name=dyn_hair_04_mid
  swing1LimitDegrees=45
  swing1Motion=Free
  swing2LimitDegrees=45
  swing2Motion=Free
  twistLimitDegrees=45
  twistMotion=Free
 }
 {
  bone1Name=dyn_hair_back_04_l
  bone2Name=dyn_hair_04_mid
  swing1LimitDegrees=45
  swing1Motion=Free
  swing2LimitDegrees=45
  swing2Motion=Free
  twistLimitDegrees=45
  twistMotion=Free
 }
]
