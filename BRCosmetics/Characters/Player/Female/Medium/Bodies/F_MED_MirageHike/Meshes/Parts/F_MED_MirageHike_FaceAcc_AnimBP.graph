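# Animation Blueprint driving secondary hair motion on the MirageHike female FaceAcc part.
# Parent mesh pose passes through bone constraints, AnimDynamics, an emote-fed Control Rig, fallback constraints for disabled rigid bodies, and jump/parachute offsets.
# That cached pose feeds two double-RigidBody chains, blended by front-end state; the front-end chain adds corrective bone offsets.
# The event graph updates gravity, zeroing inertia scales and suppressing jump offsets when CheckForBackbling is true.
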AnimBlueprint /BRCosmetics/Characters/Player/Female/Medium/Bodies/F_MED_MirageHike/Meshes/Parts/F_MED_MirageHike_FaceAcc_AnimBP
target_skeleton=/Game/Characters/Player/Female/Medium/Bodies/F_MED_MirageHike/Meshes/Parts/F_MED_MirageHike_FaceAcc_Skeleton
graphs [
 AnimGraph {nodes [
   n1 = NOT Boolean(A=Get bRigidBodyEnabled())
   "Constraint - dyn_hair_back_02_l (PP)"!(ComponentPose, bAlphaBoolEnabled:bool=@n1, ConstraintWeights_0=0.2, ConstraintWeights_1:float=0.8) -> (Pose) {
    alphaInputType=Bool
    alphaBoolBlend {blendInTime=0.1, blendOption=Linear}
    boneToModify {boneName=dyn_hair_back_02_l}
    constraintSetup [{targetBone {boneName=spine_05}}, {targetBone {boneName=clavicle_l}}]
    constraintWeights [1,1]
   }
   "Constraint - dyn_hair_back_02_r (PP)"!(ComponentPose, bAlphaBoolEnabled:bool=@n1, ConstraintWeights_0=0.6, ConstraintWeights_1:float=0.4) -> (Pose) {
    alphaInputType=Bool
    alphaBoolBlend {blendInTime=0.1, blendOption=Linear}
    boneToModify {boneName=dyn_hair_back_02_r}
    constraintSetup [{targetBone {boneName=spine_05}}, {targetBone {boneName=dyn_hair_02_mid}}]
    constraintWeights [1,1]
   }
   "Constraint - dyn_hair_back_04_r (PP)"!(ComponentPose, bAlphaBoolEnabled:bool=True, ConstraintWeights_0=0.9, ConstraintWeights_1:float=0.1) -> (Pose) {
    alphaInputType=Bool
    alphaBoolBlend {blendInTime=0.1, blendOption=Linear}
    boneToModify {boneName=dyn_hair_back_04_r}
    constraintSetup [{targetBone {boneName=spine_04}}, {targetBone {boneName=spine_05}}]
    constraintWeights [1,1]
   }
   n2 = "Constraint - dyn_hair_back_04_l (P)"(Alpha=SelectFloat(A=0.6, B=0, bPickA=Get bIsPlayingEmote()), ConstraintWeights_0=1) {
    alphaBoolBlend {blendOption=Linear}
    boneToModify {boneName=dyn_hair_back_04_l}
    constraintSetup [{targetBone {boneName=dyn_hair_back_04_fl}}]
    constraintWeights [1]
   }
   "Constraint - dyn_hair_04_mid (P)"(ComponentPose=@n2, Alpha=SelectFloat(A=0.4, B=0, bPickA=Get bIsPlayingEmote()), ConstraintWeights_0=1) {
    alphaBoolBlend {blendOption=Linear}
    boneToModify {boneName=dyn_hair_04_mid}
    constraintSetup [{targetBone {boneName=dyn_hair_back_04_l}}]
    constraintWeights [1]
   }
   "Transform (Modify) Bone - Bone: dyn_hair_bang_02_right_mstr"(Alpha=SelectFloat(A=1, B=0, bPickA=Get bIsPlayingEmote()), Translation="-2,0,0", Rotation="0,0,0", Scale="1,1,1") {
    rotationMode=BMM_Additive
    rotationSpace=BCS_BoneSpace
    scaleMode=BMM_Additive
    scaleSpace=BCS_BoneSpace
    translationMode=BMM_Additive
    translationSpace=BCS_BoneSpace
    alphaBoolBlend {blendOption=Linear}
    boneToModify {boneName=dyn_hair_bang_02_right_mstr}
    translation {x=-2}
   }
   n3 = RigidBody!(
    ComponentPose=Local To Component(LocalPose=Use cached pose 'presim'())
    Alpha
    OverrideWorldGravity:Vector=Get GravityOverride()
    ExternalForce
    ComponentLinearAccScale:Vector=Get ACC() -> (ACC:Vector)
    ComponentLinearVelScale:Vector=Get VEL() -> (VEL:Vector)
   ) -> (Pose) {
    bClampLinearTranslationLimitToRefPose=true
    bOverrideWorldGravity=true
    overridePhysicsAsset=&/F_MED_MirageHike_FrontEnd_FaceAcc_Physics1
    simulationSpace=BaseBoneSpace
    alphaBoolBlend {blendOption=Linear}
    baseBoneRef {boneName=spine_05}
    componentAppliedLinearAccClamp=<2500 2500 2500>
    componentLinearAccScale=<0.3 0.3 0.5>
    componentLinearVelScale=<0.3 0.3 0.5>
    overrideWorldGravity {z=-800}
    simSpaceSettings {
     maxAngularAcceleration=1000
     maxAngularVelocity=3
     maxLinearAcceleration=700
     maxLinearVelocity=500
     velocityScaleZ=0
     worldAlpha=1
     externalLinearDragV=<0.2 0.2>
    }
   }
   n4 = RigidBody!(ComponentPose=@n3, Alpha, OverrideWorldGravity="0,0,-980", ExternalForce) -> (Pose) {
    bClampLinearTranslationLimitToRefPose=true
    bOverrideWorldGravity=true
    overridePhysicsAsset=&/F_MED_MirageHike_FR_FaceAcc_Physics1
    simulationSpace=BaseBoneSpace
    alphaBoolBlend {blendOption=Linear}
    baseBoneRef {boneName=spine_05}
    componentAppliedLinearAccClamp=<1000 1000 1000>
    componentLinearAccScale=<0.3 0.3 0.3>
    componentLinearVelScale=<0.3 0.3 0.3>
   }
   n5 = "Transform (Modify) Bone - Bone: dyn_hair_back_03_mstr_l"(ComponentPose=@n4, Alpha=1, Translation="-0.296,-1.154,2.216", Rotation="1.033,-2.807,-0.017", Scale="1.000,0.721,1.000") {
    rotationMode=BMM_Additive
    rotationSpace=BCS_BoneSpace
    scaleMode=BMM_Additive
    scaleSpace=BCS_BoneSpace
    translationMode=BMM_Additive
    translationSpace=BCS_BoneSpace
    alphaBoolBlend {blendOption=Linear}
    alphaScaleBiasClamp {bInterpResult=true, interpSpeedIncreasing=3}
    boneToModify {boneName=dyn_hair_back_03_mstr_l}
    rotation=<rot 1.033 -2.807 -0.017>
    scale {y=0.7212}
    translation=<-0.296 -1.154 2.216>
   }
   n6 = "Transform (Modify) Bone - Bone: dyn_hair_back_03_r"(ComponentPose=@n5, Alpha=1, Translation="0,0,0", Rotation="-0.012,-4.090,0.007", Scale="1.000,0.641,1.000") {
    rotationMode=BMM_Additive
    rotationSpace=BCS_BoneSpace
    scaleMode=BMM_Additive
    scaleSpace=BCS_BoneSpace
    translationMode=BMM_Additive
    translationSpace=BCS_BoneSpace
    alphaBoolBlend {blendOption=Linear}
    alphaScaleBiasClamp {bInterpResult=true, interpSpeedIncreasing=3}
    boneToModify {boneName=dyn_hair_back_03_r}
    rotation=<rot -0.012 -4.09 0.007>
    scale {y=0.6411}
   }
   n7 = "Transform (Modify) Bone - Bone: dyn_hair_back_04_l"(ComponentPose=@n6, Alpha=1, Translation="0.153,0.083,-1.624", Rotation="5.084,-0.322,0.627", Scale="1,1,1") {
    rotationMode=BMM_Additive
    rotationSpace=BCS_BoneSpace
    translationMode=BMM_Additive
    translationSpace=BCS_BoneSpace
    alphaBoolBlend {blendOption=Linear}
    alphaScaleBiasClamp {bInterpResult=true, interpSpeedIncreasing=3}
    boneToModify {boneName=dyn_hair_back_04_l}
    rotation=<rot 5.084 -0.322 0.627>
    translation=<0.153 0.083 -1.624>
   }
   n8 = "Transform (Modify) Bone - Bone: dyn_hair_back_05_l"(ComponentPose=@n7, Alpha=1, Translation="0,0,0", Rotation="4.912,4.908,0.992", Scale="1,1,1") {
    rotationMode=BMM_Additive
    rotationSpace=BCS_BoneSpace
    translationMode=BMM_Additive
    translationSpace=BCS_BoneSpace
    alphaBoolBlend {blendOption=Linear}
    alphaScaleBiasClamp {bInterpResult=true, interpSpeedIncreasing=3}
    boneToModify {boneName=dyn_hair_back_05_l}
    rotation=<rot 4.912 4.908 0.992>
   }
   n9 = "Transform (Modify) Bone - Bone: dyn_hair_back_04_fl"(ComponentPose=@n8, Alpha=1, Translation="0,0,0", Rotation="4.814,7.202,1.167", Scale="1,1,1") {
    rotationMode=BMM_Additive
    rotationSpace=BCS_BoneSpace
    translationMode=BMM_Additive
    translationSpace=BCS_BoneSpace
    alphaBoolBlend {blendOption=Linear}
    alphaScaleBiasClamp {bInterpResult=true, interpSpeedIncreasing=3}
    boneToModify {boneName=dyn_hair_back_04_fl}
    rotation=<rot 4.814 7.202 1.167>
   }
   n10 = "Transform (Modify) Bone - Bone: dyn_hair_back_05_fl"(ComponentPose=@n9, Alpha=1, Translation="0,0,0", Rotation="4.912,4.908,0.992", Scale="1,1,1") {
    rotationMode=BMM_Additive
    rotationSpace=BCS_BoneSpace
    translationMode=BMM_Additive
    translationSpace=BCS_BoneSpace
    alphaBoolBlend {blendOption=Linear}
    alphaScaleBiasClamp {bInterpResult=true, interpSpeedIncreasing=3}
    boneToModify {boneName=dyn_hair_back_05_fl}
    rotation=<rot 4.912 4.908 0.992>
   }
   n11 = RigidBody!(
    ComponentPose=Local To Component(LocalPose=Use cached pose 'presim'())
    Alpha
    OverrideWorldGravity=Get GravityOverride()
    ExternalForce
    ComponentLinearAccScale:Vector=Get ACC()
    ComponentLinearVelScale:Vector=Get VEL()
   ) -> (Pose) {
    bClampLinearTranslationLimitToRefPose=true
    bOverrideWorldGravity=true
    overridePhysicsAsset=&/F_MED_MirageHike_FaceAcc_Physics
    simulationSpace=BaseBoneSpace
    alphaBoolBlend {blendOption=Linear}
    baseBoneRef {boneName=spine_05}
    componentAppliedLinearAccClamp=<2500 2500 2500>
    componentLinearAccScale=<0.3 0.3 0.5>
    componentLinearVelScale=<0.3 0.3 0.5>
    overrideWorldGravity {z=-800}
   }
   n12 = RigidBody!(ComponentPose=@n11, Alpha, OverrideWorldGravity="0,0,-980", ExternalForce) -> (Pose) {
    bClampLinearTranslationLimitToRefPose=true
    bOverrideWorldGravity=true
    overridePhysicsAsset=&F_MED_MirageHike_FR_FaceAcc_Physics1
    simulationSpace=BaseBoneSpace
    alphaBoolBlend {blendOption=Linear}
    baseBoneRef {boneName=spine_05}
    componentAppliedLinearAccClamp=<1000 1000 1000>
    componentLinearAccScale=<0.3 0.3 0.3>
    componentLinearVelScale=<0.3 0.3 0.3>
   }
   Output Pose(Result=Blend Poses by bool(BlendPose_0=Component To Local(ComponentPose=@n10), BlendPose_1=Component To Local(ComponentPose=@n12), bActiveValue=Get bIsInFrontEnd()) {blendType=Linear}) {layerGroup=None}
   n13 = "Constraint - dyn_hair_01_mid (PP)"!(
    ComponentPose=Local To Component(LocalPose=Copy Pose From Mesh() {bUseAttachedParent=true})
    Alpha=1
    ConstraintWeights_0=0.6
    ConstraintWeights_1:float=0.4
   ) -> (Pose) {
    alphaBoolBlend {blendOption=Linear}
    boneToModify {boneName=dyn_hair_01_mid}
    constraintSetup [{targetBone {boneName=head}}, {targetBone {boneName=spine_05}}]
    constraintWeights [1,1]
   }
   n14 = "Constraint - dyn_hair_02_mid (PP)"!(ComponentPose=@n13, Alpha=1, ConstraintWeights_0=0.2, ConstraintWeights_1:float=0.8) -> (Pose) {
    alphaBoolBlend {blendOption=Linear}
    boneToModify {boneName=dyn_hair_02_mid}
    constraintSetup [{targetBone {boneName=head}}, {targetBone {boneName=spine_05}}]
    constraintWeights [1,1]
   }
   n15 = "Constraint - dyn_hair_03_mstr_mid (PP)"!(ComponentPose=@n14, Alpha=1, ConstraintWeights_0=0.1, ConstraintWeights_1:float=0.9) -> (Pose) {
    alphaBoolBlend {blendOption=Linear}
    boneToModify {boneName=dyn_hair_03_mstr_mid}
    constraintSetup [{targetBone {boneName=head}}, {targetBone {boneName=spine_05}}]
    constraintWeights [1,1]
   }
   n16 = "Constraint - dyn_hair_back_01_l (PP)"!(ComponentPose=@n15, Alpha=1, ConstraintWeights_0=0.3, ConstraintWeights_1:float=0.7) -> (Pose) {
    alphaBoolBlend {blendOption=Linear}
    boneToModify {boneName=dyn_hair_back_01_l}
    constraintSetup [{targetBone {boneName=head}}, {targetBone {boneName=dyn_hair_01_mid}}]
    constraintWeights [1,1]
   }
   n17 = "Constraint - dyn_hair_back_02_l (PP)"!(ComponentPose=@n16, Alpha:float=1, ConstraintWeights_0=0.3, ConstraintWeights_1=0.7) -> (Pose) {
    alphaBoolBlend {blendOption=Linear}
    boneToModify {boneName=dyn_hair_back_02_l}
    constraintSetup [{targetBone {boneName=spine_05}}, {targetBone {boneName=clavicle_l}}]
    constraintWeights [1,1]
   }
   n18 = "Constraint - dyn_hair_back_03_mstr_l (PP)"!(ComponentPose=@n17, Alpha=1, ConstraintWeights_0=0.3, ConstraintWeights_1:float=0.7) -> (Pose) {
    alphaBoolBlend {blendOption=Linear}
    boneToModify {boneName=dyn_hair_back_03_mstr_l}
    constraintSetup [{targetBone {boneName=spine_05}}, {targetBone {boneName=clavicle_l}}]
    constraintWeights [1,1]
   }
   n19 = "Constraint - dyn_hair_back_01_r (PP)"!(ComponentPose=@n18, Alpha=1, ConstraintWeights_0=0.3, ConstraintWeights_1:float=0.7) -> (Pose) {
    alphaBoolBlend {blendOption=Linear}
    boneToModify {boneName=dyn_hair_back_01_r}
    constraintSetup [{targetBone {boneName=head}}, {targetBone {boneName=dyn_hair_01_mid}}]
    constraintWeights [1,1]
   }
   n20 = "Constraint - dyn_hair_back_02_r (PP)"!(ComponentPose=@n19, Alpha:float=1, ConstraintWeights_0=0.3, ConstraintWeights_1=0.7) -> (Pose) {
    alphaBoolBlend {blendOption=Linear}
    boneToModify {boneName=dyn_hair_back_02_r}
    constraintSetup [{targetBone {boneName=spine_05}}, {targetBone {boneName=clavicle_r}}]
    constraintWeights [1,1]
   }
   n21 = "Constraint - dyn_hair_back_03_mstr_r (PP)"!(ComponentPose=@n20, Alpha=1, ConstraintWeights_0=0.3, ConstraintWeights_1:float=0.7) -> (Pose) {
    alphaBoolBlend {blendOption=Linear}
    boneToModify {boneName=dyn_hair_back_03_mstr_r}
    constraintSetup [{targetBone {boneName=spine_05}}, {targetBone {boneName=clavicle_r}}]
    constraintWeights [1,1]
   }
   n22 = "Constraint - dyn_hair_right_01 (PP)"!(ComponentPose=@n21, Alpha=1, ConstraintWeights_0=0.6, ConstraintWeights_1:float=0.4) -> (Pose) {
    alphaBoolBlend {blendOption=Linear}
    boneToModify {boneName=dyn_hair_right_01}
    constraintSetup [{targetBone {boneName=spine_05}}, {targetBone {boneName=head}}]
    constraintWeights [1,1]
   }
   n23 = "Constraint - dyn_hair_left_01 (PP)"!(ComponentPose=@n22, Alpha=1, ConstraintWeights_0=0.4, ConstraintWeights_1:float=0.6) -> (Pose) {
    alphaBoolBlend {blendOption=Linear}
    boneToModify {boneName=dyn_hair_left_01}
    constraintSetup [{targetBone {boneName=spine_05}}, {targetBone {boneName=head}}]
    constraintWeights [1,1]
   }
   n24 = "Constraint - dyn_hair_right_02 (PP)"!(ComponentPose=@n23, Alpha=1, ConstraintWeights_0=1, ConstraintWeights_1:float=0) -> (Pose) {
    alphaBoolBlend {blendOption=Linear}
    boneToModify {boneName=dyn_hair_right_02}
    constraintSetup [{targetBone {boneName=spine_05}}, {targetBone {boneName=head}}]
    constraintWeights [1,1]
   }
   n25 = "Constraint - dyn_hair_left_02 (PP)"!(ComponentPose=@n24, Alpha=1, ConstraintWeights_0=1, ConstraintWeights_1:float=0) -> (Pose) {
    alphaBoolBlend {blendOption=Linear}
    boneToModify {boneName=dyn_hair_left_02}
    constraintSetup [{targetBone {boneName=spine_05}}, {targetBone {boneName=head}}]
    constraintWeights [1,1]
   }
   n26 = Anim Dynamics - Bone: dyn_hair_left_02(ComponentPose=@n25, Alpha=1, ExternalForce="0,0,0") {
    angularDampingOverride=0.98
    angularSpringConstant=50
    bAngularSpring=true
    bGravityOverrideInSimSpace=true
    bOverrideAngularDamping=true
    simulationSpace=RootRelative
    alphaBoolBlend {blendOption=Linear}
    boundBone {boneName=dyn_hair_left_02}
    componentAppliedLinearAccClamp=<3500 3500 3500>
    componentLinearAccScale=<0.6 0.6 1.2>
    componentLinearVelScale=<0.6 0.6 1.2>
    physicsBodyDefinitions [{
      sphereCollisionRadius=0
      boundBone {boneName=dyn_hair_left_02}
      boxExtents=<15 12 12>
      constraintSetup {angularLimitsMax=<1 0 15>, angularLimitsMin=<-1 -15 0>, angularTarget=<1 0 0>}
      localJointOffset=<1 0 0>
     }]
    relativeSpaceBone {boneName=root}
   }
   n27 = Anim Dynamics - Bone: dyn_hair_left_03(ComponentPose=@n26, Alpha=1, ExternalForce="0,0,0") {
    angularDampingOverride=0.98
    angularSpringConstant=50
    bAngularSpring=true
    bGravityOverrideInSimSpace=true
    bOverrideAngularDamping=true
    simulationSpace=RootRelative
    alphaBoolBlend {blendOption=Linear}
    boundBone {boneName=dyn_hair_left_03}
    componentAppliedLinearAccClamp=<3500 3500 3500>
    componentLinearAccScale=<0.6 0.6 1.2>
    componentLinearVelScale=<0.6 0.6 1.2>
    physicsBodyDefinitions [{
      sphereCollisionRadius=0
      boundBone {boneName=dyn_hair_left_03}
      boxExtents=<15 12 12>
      constraintSetup {angularLimitsMax=<1 0 10>, angularLimitsMin=<-1 -15 0>, angularTarget=<1 0 0>}
      localJointOffset=<1 0 0>
     }]
    relativeSpaceBone {boneName=root}
   }
   n28 = Anim Dynamics - Bone: dyn_hair_right_02(ComponentPose=@n27, Alpha=1, ExternalForce="0,0,0") {
    angularDampingOverride=0.98
    angularSpringConstant=50
    bAngularSpring=true
    bGravityOverrideInSimSpace=true
    bOverrideAngularDamping=true
    simulationSpace=RootRelative
    alphaBoolBlend {blendOption=Linear}
    boundBone {boneName=dyn_hair_right_02}
    componentAppliedLinearAccClamp=<3500 3500 3500>
    componentLinearAccScale=<0.6 0.6 1.2>
    componentLinearVelScale=<0.6 0.6 1.2>
    physicsBodyDefinitions [{
      sphereCollisionRadius=0
      boundBone {boneName=dyn_hair_right_02}
      boxExtents=<15 12 12>
      constraintSetup {angularLimitsMax=<1 0 10>, angularLimitsMin=<-1 -15 0>, angularTarget=<1 0 0>}
      localJointOffset=<1 0 0>
     }]
    relativeSpaceBone {boneName=root}
   }
   n29 = Anim Dynamics - Bone: dyn_hair_right_03(ComponentPose=@n28, Alpha=1, ExternalForce="0,0,0") {
    angularDampingOverride=0.98
    angularSpringConstant=50
    bAngularSpring=true
    bGravityOverrideInSimSpace=true
    bOverrideAngularDamping=true
    simulationSpace=RootRelative
    alphaBoolBlend {blendOption=Linear}
    boundBone {boneName=dyn_hair_right_03}
    componentAppliedLinearAccClamp=<3500 3500 3500>
    componentLinearAccScale=<0.6 0.6 1.2>
    componentLinearVelScale=<0.6 0.6 1.2>
    physicsBodyDefinitions [{
      sphereCollisionRadius=0
      boundBone {boneName=dyn_hair_right_03}
      boxExtents=<15 12 12>
      constraintSetup {angularLimitsMax=<1 0 10>, angularLimitsMin=<-1 -15 0>, angularTarget=<1 0 0>}
      localJointOffset=<1 0 0>
     }]
    relativeSpaceBone {boneName=root}
   }
   n30 = Control Rig!(Source=Component To Local(ComponentPose=@n29), Alpha, Emote:double=Get Emote() -> (Emote:double)) -> (Pose) {
    alphaBoolBlend {blendOption=Linear}
    controlRigAssetReference {blueprintRigClass=&/F_MED_MirageHike_FaceAcc_CtrlRig.F_MED_MirageHike_FaceAcc_CtrlRig_C}
   }
   n31 = "Constraint - dyn_hair_02_mid (PP)"!(ComponentPose=Local To Component(LocalPose=@n30), bAlphaBoolEnabled:bool=@n1, ConstraintWeights_0=0.1, ConstraintWeights_1=0.9) -> (Pose) {
    alphaInputType=Bool
    alphaBoolBlend {blendInTime=0.1, blendOption=Linear, blendOutTime=0.1}
    boneToModify {boneName=dyn_hair_02_mid}
    constraintSetup [{targetBone {boneName=head}}, {targetBone {boneName=spine_05}}]
    constraintWeights [1,1]
   }
   n32 = "Constraint - dyn_hair_03_mstr_mid (PP)"!(ComponentPose=@n31, bAlphaBoolEnabled:bool=@n1, ConstraintWeights_0=0.1, ConstraintWeights_1=0.9) -> (Pose) {
    alphaInputType=Bool
    alphaBoolBlend {blendInTime=0.1, blendOption=Linear, blendOutTime=0.1}
    boneToModify {boneName=dyn_hair_03_mstr_mid}
    constraintSetup [{targetBone {boneName=head}}, {targetBone {boneName=spine_05}}]
    constraintWeights [1,1]
   }
   n33 = "Constraint - dyn_hair_04_mid (PP)"!(ComponentPose=@n32, bAlphaBoolEnabled:bool=@n1, ConstraintWeights_0=0.1, ConstraintWeights_1:float=0.9) -> (Pose) {
    alphaInputType=Bool
    alphaBoolBlend {blendInTime=0.1, blendOption=Linear, blendOutTime=0.1}
    boneToModify {boneName=dyn_hair_04_mid}
    constraintSetup [{targetBone {boneName=spine_05}}, {targetBone {boneName=spine_04}}]
    constraintWeights [1,1]
   }
   n34 = "Constraint - dyn_hair_back_03_mstr_l (PP)"!(ComponentPose=@n33, bAlphaBoolEnabled:bool=@n1, ConstraintWeights_0=0.3, ConstraintWeights_1=0.7) -> (Pose) {
    alphaInputType=Bool
    alphaBoolBlend {blendInTime=0.1, blendOption=Linear}
    boneToModify {boneName=dyn_hair_back_03_mstr_l}
    constraintSetup [{targetBone {boneName=spine_05}}, {targetBone {boneName=clavicle_l}}]
    constraintWeights [1,1]
   }
   n35 = "Constraint - dyn_hair_back_04_l (PP)"!(ComponentPose=@n34, bAlphaBoolEnabled:bool=@n1, ConstraintWeights_0=0.9, ConstraintWeights_1:float=0.1) -> (Pose) {
    alphaInputType=Bool
    alphaBoolBlend {blendInTime=0.1, blendOption=Linear}
    boneToModify {boneName=dyn_hair_back_04_l}
    constraintSetup [{targetBone {boneName=spine_04}}, {targetBone {boneName=spine_05}}]
    constraintWeights [1,1]
   }
   n36 = "Constraint - dyn_hair_back_04_fl (PP)"!(ComponentPose=@n35, bAlphaBoolEnabled:bool=@n1, ConstraintWeights_0=0.6, ConstraintWeights_1:float=0.4) -> (Pose) {
    alphaInputType=Bool
    alphaBoolBlend {blendInTime=0.1, blendOption=Linear}
    boneToModify {boneName=dyn_hair_back_04_fl}
    constraintSetup [{targetBone {boneName=spine_04}}, {targetBone {boneName=spine_05}}]
    constraintWeights [1,1]
   }
   n37 = "Constraint - dyn_hair_back_03_mstr_r (PP)"!(ComponentPose=@n36, bAlphaBoolEnabled:bool=@n1, ConstraintWeights_0=0.3, ConstraintWeights_1=0.7) -> (Pose) {
    alphaInputType=Bool
    alphaBoolBlend {blendInTime=0.1, blendOption=Linear}
    boneToModify {boneName=dyn_hair_back_03_mstr_r}
    constraintSetup [{targetBone {boneName=spine_05}}, {targetBone {boneName=clavicle_r}}]
    constraintWeights [1,1]
   }
   n38 = "Constraint - dyn_hair_back_04_r (PP)"(ComponentPose=@n37, bAlphaBoolEnabled=@n1, ConstraintWeights_0=0.9, ConstraintWeights_1=0.1) {
    alphaInputType=Bool
    alphaBoolBlend {blendInTime=0.1, blendOption=Linear}
    boneToModify {boneName=dyn_hair_back_04_r}
    constraintSetup [{targetBone {boneName=spine_04}}, {targetBone {boneName=spine_05}}]
    constraintWeights [1,1]
   }
   n39 = "Transform (Modify) Bone - Bone: dyn_hair_back_03_r"!(ComponentPose=@n38, bAlphaBoolEnabled:bool=@n1, Translation="0,0,0", Rotation="0,10,0", Scale="1,1,1") -> (Pose) {
    alphaInputType=Bool
    rotationMode=BMM_Additive
    rotationSpace=BCS_BoneSpace
    translationMode=BMM_Additive
    translationSpace=BCS_BoneSpace
    alphaBoolBlend {blendOption=Linear}
    alphaScaleBiasClamp {bInterpResult=true, interpSpeedIncreasing=3}
    boneToModify {boneName=dyn_hair_back_03_r}
    rotation {yaw=10}
   }
   n40 = "Transform (Modify) Bone - Bone: dyn_hair_back_03_l"!(ComponentPose=@n39, bAlphaBoolEnabled:bool=@n1, Translation="0,0,0", Rotation="0,-10,0", Scale="1,1,1") -> (Pose) {
    alphaInputType=Bool
    rotationMode=BMM_Additive
    rotationSpace=BCS_BoneSpace
    translationMode=BMM_Additive
    translationSpace=BCS_BoneSpace
    alphaBoolBlend {blendOption=Linear}
    alphaScaleBiasClamp {bInterpResult=true, interpSpeedIncreasing=3}
    boneToModify {boneName=dyn_hair_back_03_l}
    rotation {yaw=-10}
   }
   n41 = Get jumping() -> (jumping:double)
   n42 = "Transform (Modify) Bone - Bone: dyn_hair_back_01_r"(ComponentPose=@n40, Alpha=@n41, Translation="0,0,0", Rotation="20,0,0", Scale="1,1,1") {
    rotationMode=BMM_Additive
    rotationSpace=BCS_BoneSpace
    translationMode=BMM_Additive
    translationSpace=BCS_BoneSpace
    alphaBoolBlend {blendOption=Linear}
    alphaScaleBiasClamp {bInterpResult=true, interpSpeedIncreasing=3}
    boneToModify {boneName=dyn_hair_back_01_r}
    rotation {pitch=20}
   }
   n43 = "Transform (Modify) Bone - Bone: dyn_hair_01_mid"(ComponentPose=@n42, Alpha=@n41, Translation="0,0,0", Rotation="20,0,0", Scale="1,1,1") {
    rotationMode=BMM_Additive
    rotationSpace=BCS_BoneSpace
    translationMode=BMM_Additive
    translationSpace=BCS_BoneSpace
    alphaBoolBlend {blendOption=Linear}
    alphaScaleBiasClamp {bInterpResult=true, interpSpeedIncreasing=3}
    boneToModify {boneName=dyn_hair_01_mid}
    rotation {pitch=20}
   }
   n44 = "Transform (Modify) Bone - Bone: dyn_hair_back_01_l"(ComponentPose=@n43, Alpha=@n41, Translation="0,0,0", Rotation="20,0,0", Scale="1,1,1") {
    rotationMode=BMM_Additive
    rotationSpace=BCS_BoneSpace
    translationMode=BMM_Additive
    translationSpace=BCS_BoneSpace
    alphaBoolBlend {blendOption=Linear}
    alphaScaleBiasClamp {bInterpResult=true, interpSpeedIncreasing=3}
    boneToModify {boneName=dyn_hair_back_01_l}
    rotation {pitch=20}
   }
   n45 = Get bIsParachuteOpened()
   n46 = "Transform (Modify) Bone - Bone: dyn_hair_left_01"!(ComponentPose=@n44, bAlphaBoolEnabled:bool=@n45, Translation="0,0,0", Rotation="15,0,0", Scale="1,1,1") -> (Pose) {
    alphaInputType=Bool
    rotationMode=BMM_Additive
    rotationSpace=BCS_BoneSpace
    translationMode=BMM_Additive
    translationSpace=BCS_BoneSpace
    alphaBoolBlend {blendInTime=0.1, blendOption=Linear, blendOutTime=0.1}
    alphaScaleBiasClamp {bInterpResult=true, interpSpeedIncreasing=3}
    boneToModify {boneName=dyn_hair_left_01}
    rotation {pitch=15}
   }
   n47 = "Transform (Modify) Bone - Bone: dyn_hair_right_01"!(ComponentPose=@n46, bAlphaBoolEnabled:bool=@n45, Translation="0,0,0", Rotation="15,0,0", Scale="1,1,1") -> (Pose) {
    alphaInputType=Bool
    rotationMode=BMM_Additive
    rotationSpace=BCS_BoneSpace
    translationMode=BMM_Additive
    translationSpace=BCS_BoneSpace
    alphaBoolBlend {blendInTime=0.1, blendOption=Linear, blendOutTime=0.1}
    alphaScaleBiasClamp {bInterpResult=true, interpSpeedIncreasing=3}
    boneToModify {boneName=dyn_hair_right_01}
    rotation {pitch=15}
   }
   Save cached pose 'presim'(Pose=Component To Local(ComponentPose=@n47)) {cacheName=presim}
  ]}
 EventGraph {nodes [
   Event BlueprintUpdateAnimation()
   > n1 = Sequence()
   > n2 = Gravity Override State Function DEPRECATED(CharacterPartAnimInstance=Self-Reference(), GameTimeInSeconds=GetGameTimeInSeconds(), GravityOverrideParams=Get Gravity Override Params())
   > Set GravityOverride(GravityOverride=@n2)
   > n3 = Branch(Condition=CheckForBackbling())
   > Set ACC(ACC:Vector="0, 0, 0") -> (Output_Get:Vector)
   > Set VEL(VEL:Vector="0, 0, 0") -> (Output_Get:Vector)
   > Set backpack(backpack:double=0.5) -> (Output_Get:double)
   Set ACC(execute=@n3.else, ACC="0.5,0.5,0.5")
   > Set VEL(VEL="0.5,0.5,0.5")
   > Set backpack(backpack=0)
   Set Emote(execute=@n1.then_1, Emote:double="To Float (Boolean)"(InBool=Get bIsPlayingEmote())) -> (Output_Get:double)
   n4 = Event OnBeginFalling()
   n5 = Branch(execute=[@n4.then, @n7.then], Condition=CheckForBackbling())
   Set jumping(execute=[@n6.then, @n8.then, @n5.then], jumping:double=0) -> (Output_Get:double)
   Set jumping(execute=@n5.else, jumping=1)
   n6 = Event OnEndFalling()
   n7 = Event OnBeginJumping()
   n8 = Event OnEndJumping()
  ]}
]
variables [
 ACC {type=Vector}
 backpack {type=number}
 Emote {type=number}
 Gravity Override Params {type=GravityOverrideParamsStruct}
 GravityOverride {type=Vector}
 jumping {type=number}
 VEL {type=Vector}
]
