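# Physics asset for the MirageHike female character's face accessory part, providing collision and joints for dynamic hair.
# Capsules on pelvis, spine, neck and head form a body chain, with a hair master hanging off the head.
# From that master run middle, left and right side, and two back chains, each back chain forking near its tip.
# Body and master joints rotate freely, while strand joints use narrower swing and twist limits.
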PhysicsAsset /BRCosmetics/Characters/Player/Female/Medium/Bodies/F_MED_MirageHike/Meshes/Parts/F_MED_MirageHike_FaceAcc_PA
bodies {
 dfrm_hair_mstr [{shapeType=Capsule, length=5, radius=4.5, center=<0 -2 5>, rotation=<rot 0 0 10>}]
 dyn_hair_01_mid [{shapeType=Capsule, length=5, radius=8, shapeName=dyn_hair_01_mid_capsule, center=<6 0 6.5>, rotation=<rot 90 0 0>}]
 dyn_hair_02_mid [{shapeType=Capsule, length=6, radius=8, shapeName=dyn_hair_02_mid_capsule, center=<6 0 5.5>, rotation=<rot 90 0 0>}]
 dyn_hair_03_mid [{shapeType=Capsule, length=8, radius=9, shapeName=dyn_hair_03_mid_capsule, center=<5.5 0 5.5>, rotation=<rot 90 0 0>}]
 dyn_hair_04_mid [{shapeType=Capsule, length=4, radius=8, shapeName=dyn_hair_04_mid_capsule, center=<6 0 5.5>, rotation=<rot 90 0 0>}]
 dyn_hair_05_mid [{shapeType=Capsule, length=6, radius=8, shapeName=dyn_hair_05_mid_capsule, center=<7 0 5.5>, rotation=<rot 90 0 0>}]
 dyn_hair_back_01_l [{shapeType=Capsule, length=4, radius=8, shapeName=dyn_hair_back_01_l_capsule, center=<6.1263 0 6.5>, rotation=<rot 90 0 0>}]
 dyn_hair_back_01_r [{shapeType=Capsule, length=5, radius=8, shapeName=dyn_hair_back_01_r_capsule, center=<6.5549 0 6.5>, rotation=<rot 90 0 0>}]
 dyn_hair_back_02_l [{shapeType=Capsule, length=4, radius=8, shapeName=dyn_hair_back_02_l_capsule, center=<5.75 0 5>, rotation=<rot 90 0 0>}]
 dyn_hair_back_02_r [{shapeType=Capsule, length=5, radius=8, shapeName=dyn_hair_back_02_r_capsule, center=<6.5 0 5.5>, rotation=<rot 90 0 0>}]
 dyn_hair_back_03_l [{shapeType=Capsule, length=6, radius=8, shapeName=dyn_hair_back_03_l_capsule, center=<5.5 0 5.5>, rotation=<rot 90 0 0>}]
 dyn_hair_back_03_r [{shapeType=Capsule, length=5, radius=8, shapeName=dyn_hair_back_03_r_capsule, center=<6 0 5.5>, rotation=<rot 90 0 0>}]
 dyn_hair_back_04_fl [{shapeType=Capsule, length=3, radius=8, shapeName=dyn_hair_back_04_fl_capsule, center=<5 0 5.5>, rotation=<rot 90 0 0>}]
 dyn_hair_back_04_fr [{shapeType=Capsule, length=3, radius=8, shapeName=dyn_hair_back_04_fr_capsule, center=<5 0 5.5>, rotation=<rot 90 0 0>}]
 dyn_hair_back_04_l [{shapeType=Capsule, length=4, radius=8, shapeName=dyn_hair_back_04_l_capsule, center=<6.75 0 5.5>, rotation=<rot 90 0 0>}]
 dyn_hair_back_04_r [{shapeType=Capsule, length=4, radius=8, shapeName=dyn_hair_back_04_r_capsule, center=<6 0 5.5>, rotation=<rot 90 0 0>}]
 dyn_hair_back_05_fl [{shapeType=Capsule, length=6, radius=8, shapeName=dyn_hair_back_05_fl_capsule, center=<8 0 5.5>, rotation=<rot 90 0 0>}]
 dyn_hair_back_05_fr [{shapeType=Capsule, length=2, radius=8, shapeName=dyn_hair_back_05_fr_capsule, center=<4 0 5.5>, rotation=<rot 90 0 0>}]
 dyn_hair_back_05_l [{shapeType=Capsule, length=6, radius=8, shapeName=dyn_hair_back_05_l_capsule, center=<8 0 5.5>, rotation=<rot 90 0 0>}]
 dyn_hair_back_05_r [{shapeType=Capsule, length=3, radius=8, shapeName=dyn_hair_back_05_r_capsule, center=<5.5 0 5.5>, rotation=<rot 90 0 0>}]
 dyn_hair_back_mstr_l [{shapeType=Sphere, radius=1, shapeName=dyn_hair_back_mstr_l_sphere}]
 dyn_hair_back_mstr_r [{shapeType=Capsule, length=0.505, radius=0.505, shapeName=dyn_hair_back_mstr_r_capsule, rotation=<rot 0 -90 0>}]
 dyn_hair_left_01 [{shapeType=Capsule, length=6, radius=2, shapeName=dyn_hair_left_01_capsule, center=<5 0 0>, rotation=<rot 90 0 0>}]
 dyn_hair_left_02 [{shapeType=Capsule, length=1.5, radius=1.5, shapeName=dyn_hair_left_02_capsule, center=<1.5 0 0>, rotation=<rot 90 0 0>}]
 dyn_hair_left_03 [{shapeType=Capsule, length=2.5, radius=1, shapeName=dyn_hair_left_03_capsule, center=<2.5 0 0>, rotation=<rot 90 0 0>}]
 dyn_hair_left_mstr [{shapeType=Sphere, radius=0.505, shapeName=dyn_hair_left_mstr_sphere}]
 dyn_hair_mid_mstr [{shapeType=Sphere, radius=1, shapeName=dyn_hair_mid_mstr_sphere}]
 dyn_hair_right_01 [{shapeType=Capsule, length=6, radius=2, shapeName=dyn_hair_right_01_capsule, center=<5 0 0>, rotation=<rot 90 0 0>}]
 dyn_hair_right_02 [{shapeType=Capsule, length=3, radius=1.5, shapeName=dyn_hair_right_02_capsule, center=<2.5 0 0>, rotation=<rot 90 0 0>}]
 dyn_hair_right_03 [{shapeType=Capsule, length=4, radius=1, shapeName=dyn_hair_right_03_capsule, center=<2.5 0 0>, rotation=<rot 90 0 0>}]
 dyn_hair_right_mstr [{shapeType=Sphere, radius=1, shapeName=dyn_hair_right_mstr_sphere}]
 head [
  {shapeType=Capsule, length=6, radius=5, shapeName=head_capsule, center=<4.5 2.5 0>, rotation=<rot 90 90 80>}
  {shapeType=Capsule, length=6.75, radius=4.25, shapeName=head_capsule, center=<2.9031 4.5 0.1485>, rotation=<rot 90 0 0>}
 ]
 neck_01 [{shapeType=Capsule, length=5, radius=5, shapeName=neck_01_capsule, rotation=<rot 90 0 0>}]
 neck_02 [{shapeType=Capsule, length=5, radius=5, shapeName=neck_01_capsule, rotation=<rot 90 0 0>}]
 pelvis [{shapeType=Capsule, length=20, radius=9, shapeName=spine_05_capsule}]
 spine_01 [{shapeType=Capsule, length=20, radius=9, shapeName=spine_01_capsule}]
 spine_02 [{shapeType=Capsule, length=20, radius=9, shapeName=spine_05_capsule}]
 spine_03 [{shapeType=Capsule, length=20, radius=9, shapeName=spine_05_capsule}]
 spine_04 [{shapeType=Capsule, length=20, radius=9, shapeName=spine_05_capsule}]
 spine_05 [{shapeType=Capsule, length=20, radius=9, shapeName=spine_05_capsule}]
}
constraints [
 {
  bone1Name=dyn_hair_back_mstr_l
  bone2Name=dfrm_hair_mstr
  swing1LimitDegrees=30
  swing1Motion=Free
  swing2LimitDegrees=30
  swing2Motion=Free
  twistLimitDegrees=2
  twistMotion=Free
 }
 {bone1Name=dyn_hair_back_01_l, bone2Name=dyn_hair_back_mstr_l, swing1LimitDegrees=15, swing2LimitDegrees=50, twistLimitDegrees=10}
 {bone1Name=dyn_hair_back_02_l, bone2Name=dyn_hair_back_01_l, swing1LimitDegrees=15, swing2LimitDegrees=50, twistLimitDegrees=10}
 {bone1Name=dyn_hair_back_03_l, bone2Name=dyn_hair_back_02_l, swing1LimitDegrees=15, swing2LimitDegrees=50, twistLimitDegrees=10}
 {bone1Name=dyn_hair_back_04_fl, bone2Name=dyn_hair_back_03_l, swing1LimitDegrees=15, swing2LimitDegrees=50, twistLimitDegrees=10}
 {bone1Name=dyn_hair_back_05_fl, bone2Name=dyn_hair_back_04_fl, swing1LimitDegrees=15, swing2LimitDegrees=50, twistLimitDegrees=10}
 {bone1Name=dyn_hair_back_04_l, bone2Name=dyn_hair_back_03_l, swing1LimitDegrees=15, swing2LimitDegrees=50, twistLimitDegrees=10}
 {bone1Name=dyn_hair_back_05_l, bone2Name=dyn_hair_back_04_l, swing1LimitDegrees=15, swing2LimitDegrees=50, twistLimitDegrees=10}
 {
  bone1Name=dyn_hair_back_mstr_r
  bone2Name=dfrm_hair_mstr
  swing1LimitDegrees=45
  swing1Motion=Free
  swing2LimitDegrees=45
  swing2Motion=Free
  twistLimitDegrees=45
  twistMotion=Free
 }
 {bone1Name=dyn_hair_back_01_r, bone2Name=dyn_hair_back_mstr_r, swing1LimitDegrees=15, swing2LimitDegrees=50, twistLimitDegrees=10}
 {bone1Name=dyn_hair_back_02_r, bone2Name=dyn_hair_back_01_r, swing1LimitDegrees=15, swing2LimitDegrees=50, twistLimitDegrees=10}
 {bone1Name=dyn_hair_back_03_r, bone2Name=dyn_hair_back_02_r, swing1LimitDegrees=15, swing2LimitDegrees=50, twistLimitDegrees=10}
 {bone1Name=dyn_hair_back_04_fr, bone2Name=dyn_hair_back_03_r, swing1LimitDegrees=15, swing2LimitDegrees=50, twistLimitDegrees=10}
 {bone1Name=dyn_hair_back_05_fr, bone2Name=dyn_hair_back_04_fr, swing1LimitDegrees=15, swing2LimitDegrees=50, twistLimitDegrees=10}
 {bone1Name=dyn_hair_back_04_r, bone2Name=dyn_hair_back_03_r, swing1LimitDegrees=15, swing2LimitDegrees=50, twistLimitDegrees=10}
 {bone1Name=dyn_hair_back_05_r, bone2Name=dyn_hair_back_04_r, swing1LimitDegrees=15, swing2LimitDegrees=50, twistLimitDegrees=10}
 {
  bone1Name=dfrm_hair_mstr
  bone2Name=head
  swing1LimitDegrees=45
  swing1Motion=Free
  swing2LimitDegrees=45
  swing2Motion=Free
  twistLimitDegrees=45
  twistMotion=Free
 }
 {
  bone1Name=spine_03
  bone2Name=spine_02
  swing1LimitDegrees=45
  swing1Motion=Free
  swing2LimitDegrees=45
  swing2Motion=Free
  twistLimitDegrees=45
  twistMotion=Free
 }
 {
  bone1Name=spine_05
  bone2Name=spine_04
  swing1LimitDegrees=45
  swing1Motion=Free
  swing2LimitDegrees=45
  swing2Motion=Free
  twistLimitDegrees=45
  twistMotion=Free
 }
 {
  bone1Name=neck_01
  bone2Name=spine_05
  swing1LimitDegrees=45
  swing1Motion=Free
  swing2LimitDegrees=45
  swing2Motion=Free
  twistLimitDegrees=45
  twistMotion=Free
 }
 {
  bone1Name=neck_02
  bone2Name=neck_01
  swing1LimitDegrees=45
  swing1Motion=Free
  swing2LimitDegrees=45
  swing2Motion=Free
  twistLimitDegrees=45
  twistMotion=Free
 }
 {
  bone1Name=head
  bone2Name=neck_02
  swing1LimitDegrees=45
  swing1Motion=Free
  swing2LimitDegrees=45
  swing2Motion=Free
  twistLimitDegrees=45
  twistMotion=Free
 }
 {
  bone1Name=dyn_hair_mid_mstr
  bone2Name=dfrm_hair_mstr
  swing1LimitDegrees=45
  swing1Motion=Free
  swing2LimitDegrees=45
  swing2Motion=Free
  twistLimitDegrees=45
  twistMotion=Free
 }
 {bone1Name=dyn_hair_01_mid, bone2Name=dyn_hair_mid_mstr, swing1LimitDegrees=15, swing2LimitDegrees=50, twistLimitDegrees=10}
 {bone1Name=dyn_hair_02_mid, bone2Name=dyn_hair_01_mid, swing1LimitDegrees=15, swing2LimitDegrees=50, twistLimitDegrees=10}
 {bone1Name=dyn_hair_04_mid, bone2Name=dyn_hair_03_mid, swing1LimitDegrees=15, swing2LimitDegrees=50, twistLimitDegrees=10}
 {bone1Name=dyn_hair_05_mid, bone2Name=dyn_hair_04_mid, swing1LimitDegrees=15, swing2LimitDegrees=50, twistLimitDegrees=10}
 {
  bone1Name=dyn_hair_left_mstr
  bone2Name=dfrm_hair_mstr
  swing1LimitDegrees=45
  swing1Motion=Free
  swing2LimitDegrees=45
  swing2Motion=Free
  twistLimitDegrees=45
  twistMotion=Free
 }
 {bone1Name=dyn_hair_left_01, bone2Name=dyn_hair_left_mstr, swing1LimitDegrees=30, swing2LimitDegrees=30, twistLimitDegrees=2}
 {bone1Name=dyn_hair_left_02, bone2Name=dyn_hair_left_01, swing1LimitDegrees=30, swing2LimitDegrees=30, twistLimitDegrees=2}
 {bone1Name=dyn_hair_left_03, bone2Name=dyn_hair_left_02, swing1LimitDegrees=30, swing2LimitDegrees=30, twistLimitDegrees=2}
 {
  bone1Name=dyn_hair_right_mstr
  bone2Name=dfrm_hair_mstr
  swing1LimitDegrees=45
  swing1Motion=Free
  swing2LimitDegrees=45
  swing2Motion=Free
  twistLimitDegrees=45
  twistMotion=Free
 }
 {bone1Name=dyn_hair_right_03, bone2Name=dyn_hair_right_02, swing1LimitDegrees=30, swing2LimitDegrees=30, twistLimitDegrees=2}
 {bone1Name=dyn_hair_right_02, bone2Name=dyn_hair_right_01, swing1LimitDegrees=30, swing2LimitDegrees=30, twistLimitDegrees=2}
 {bone1Name=dyn_hair_right_01, bone2Name=dyn_hair_right_mstr, swing1LimitDegrees=30, swing2LimitDegrees=30, twistLimitDegrees=2}
 {
  bone1Name=spine_04
  bone2Name=spine_03
  swing1LimitDegrees=45
  swing1Motion=Free
  swing2LimitDegrees=45
  swing2Motion=Free
  twistLimitDegrees=45
  twistMotion=Free
 }
 {
  bone1Name=spine_02
  bone2Name=spine_01
  swing1LimitDegrees=45
  swing1Motion=Free
  swing2LimitDegrees=45
  swing2Motion=Free
  twistLimitDegrees=45
  twistMotion=Free
 }
 {
  bone1Name=spine_01
  bone2Name=pelvis
  swing1LimitDegrees=45
  swing1Motion=Free
  swing2LimitDegrees=45
  swing2Motion=Free
  twistLimitDegrees=45
  twistMotion=Free
 }
 {bone1Name=dyn_hair_03_mid, bone2Name=dyn_hair_02_mid, swing1LimitDegrees=15, swing2LimitDegrees=50, twistLimitDegrees=10}
]
